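; Distant foliage actor that displays a configurable static mesh with a configurable foliage texture.
; Init assigns the Mesh variable and, if that succeeds, sets FoliageTexture on the slot 0 dynamic material, creating one if needed.
; The construction script unhides the actor in game and, if the component has a mesh, creates a dynamic material with the texture.
; BeginPlay, overlap and Tick events are empty, and nothing here calls Init.
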(event EventBeginPlay)

(event Collision|EventActorBeginOverlap
  (params (OtherActor :type Actor%20Object%20Reference)))

(event EventTick
  (params (DeltaSeconds :type Float%20%28single-precision%29)))

(fn Init
  (bind _staticmesh (Variables|Default|GetStaticMesh))
  (bind _texture (Variables|Default|GetTexture))
  (bind _returnvalue (Components|StaticMesh|SetStaticMesh _staticmesh (Variables|Default|GetMesh)))
  (if _returnvalue
    (bind _asmaterial_instance_dynamic (Utilities|Casting|CastToMaterialInstanceDynamic (Rendering|Material|GetMaterial _staticmesh 0))
      (:then
        (Rendering|Material|SetTextureParameterValue _asmaterial_instance_dynamic "FoliageTexture" _texture))
      (:CastFailed
        (bind _returnvalue_2 (Rendering|Material|CreateDynamicMaterialInstance _staticmesh 0 0 "None" :!class /Script/Engine.PrimitiveComponent))
        (Rendering|Material|SetTextureParameterValue _returnvalue_2 "FoliageTexture" _texture)))))

(fn ConstructionScript
  (bind _staticmesh (Variables|Default|GetStaticMesh))
  (Rendering|SetActorHiddenInGame false)
  (Utilities|IsValid (Variables|/Script/Engine.StaticMeshComponent|GetStaticMesh _staticmesh)
    (:Is%20Valid
      (bind _returnvalue (Rendering|Material|CreateDynamicMaterialInstance _staticmesh 0 0 "None" :!class /Script/Engine.PrimitiveComponent))
      (Rendering|Material|SetTextureParameterValue _returnvalue "FoliageTexture" (Variables|Default|GetTexture)))))
