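# M_DistantFoliage is a masked, dithered, fully rough material for rendering distant foliage from a texture.
# Voronoi noise distorts texture UVs and, multiplied by cell noise, breaks up opacity derived from inverted alpha.
# Camera-distance near and far fades drive visibility and near-to-far blends of distortion, opacity and albedo.
# Optional pixel depth offset combines noise with view angle, and Epic quality switches Default Lit to Subsurface shading.
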Material /Game/ArtTools/Tech_Art/DistantFoliage/M_DistantFoliage
graph {
 properties {
  bAutomaticallySetUsageInEditor=false
  bCastRayTracedShadows=false
  bFullyRough=true
  blendMode=BLEND_Masked
  ditherOpacityMask=true
  shadingModel=MSM_FromMaterialExpression
  usedShadingModels=Default Lit | Subsurface
 }
 nodes [
  Constant() {r=0.5}
  n1 = TextureCoordinate()
  n2 = TextureSample(UVs=Multiply(A=@n1, B=ScalarParameter() {defaultValue=64, parameterName=DistortTiling})) {
   samplerSource=SSM_Wrap_WorldGroupSettings
   samplerType=SAMPLERTYPE_LinearGrayscale
   texture=&/T_VoronoiNoise_64
  }
  n3 = ConstantBiasScale(None=@n2.R) {bias=-0.5, scale=2}
  n4 = Multiply(
   A=@n3
   B=LinearInterpolate(
    A=ScalarParameter() {defaultValue=0.003, parameterName=DistortStrength}
    B=ScalarParameter() {defaultValue=0.006, parameterName=DistortStrength Far}
    Alpha=NamedRerouteUsage()
   )
  ) {constB=0}
  n5 = TextureSampleParameter2D(UVs=StaticSwitchParameter(True=Add(A=@n1, B=@n4), False=@n1) {defaultValue=true, parameterName=Use Distortion}) {parameterName=FoliageTexture, samplerSource=SSM_Wrap_WorldGroupSettings, texture=&Black}
  n6 = Power(Base=OneMinus(None=@n5.A), Exp=ScalarParameter() {defaultValue=3, parameterName=OpacityPower}) {constExponent=0.8}
  n7 = LinearInterpolate(
   A=ScalarParameter() {defaultValue=3, parameterName=Opacity}
   B=ScalarParameter() {defaultValue=4, parameterName=Opacity Far}
   Alpha=NamedRerouteUsage()
  )
  n8 = TextureSample(UVs=Multiply(A=@n1, B=ScalarParameter() {defaultValue=48, parameterName=DistortTiling2})) {
   samplerSource=SSM_Wrap_WorldGroupSettings
   samplerType=SAMPLERTYPE_LinearGrayscale
   texture=&/T_CellNoise_DistantFoliage
  }
  n9 = Multiply(A=@n2.R, B=@n8.R)
  n10 = LinearInterpolate(
   A=ScalarParameter() {defaultValue=1, parameterName=OpacityDistortionPow}
   B=ScalarParameter() {defaultValue=0.8, parameterName=OpacityDistortionPow Far}
   Alpha=NamedRerouteUsage()
  )
  Multiply(A=NamedRerouteUsage(), B=Multiply(A=Multiply(A=@n6, B=@n7), B=Power(Base=@n9, Exp=@n10)))
  Multiply(
   A=Desaturation(None=@n5.RGB, Fraction=ScalarParameter() {defaultValue=-0.2, parameterName=Saturation}) {luminanceFactors=<rgba 0.2126 0.7152 0.0722 1>}
   B=LinearInterpolate(
    A=ScalarParameter() {defaultValue=1, parameterName=Albedo}
    B=ScalarParameter() {defaultValue=0.8, parameterName=Albedo Far}
    Alpha=NamedRerouteUsage()
   )
  )
  n11 = Length(None=WorldPosition() {worldPositionShaderOffset=WPT_CameraRelative}.XYZ) {is=class MaterialExpressionLength}
  n12 = Add(A=@n11, B=ScalarParameter() {defaultValue=-20000, parameterName=FarFadeOffset})
  n13 = ScalarParameter() {defaultValue=20000, parameterName=FarFadeDistance}
  NamedRerouteDeclaration(None=Saturate(None=Divide(A=@n12, B=@n13) {constB=10000})) {name=FarFade, nodeColor=<rgba 0.2656 0 1 1>}
  n14 = Add(A=@n11, B=ScalarParameter() {defaultValue=-8000, parameterName=NearFadeOffset})
  n15 = ScalarParameter() {defaultValue=2000, parameterName=NearFadeDistance}
  NamedRerouteDeclaration(None=Saturate(None=Divide(A=@n14, B=@n15) {constB=10000})) {name=NearFade, nodeColor=<rgba 0.0547 1 0 1>}
  QualitySwitch(Default=ShadingModel(), Epic=ShadingModel() {shadingModel=MSM_Subsurface})
  n16 = ScalarParameter() {defaultValue=30, parameterName=PDONoise}
  n17 = ScalarParameter() {defaultValue=75, parameterName=PDO_Low}
  n18 = ScalarParameter() {defaultValue=10, parameterName=PDO_High}
  n19 = WorldPosition() {worldPositionShaderOffset=WPT_CameraRelative}
  n20 = Saturate(None=DotProduct(A=Normalize(VectorInput=@n19.XYZ), B=Constant3Vector() {constant=<rgba 0 0 -1 0>}))
  StaticSwitchParameter(True=Add(A=Multiply(A=@n3, B=@n16), B=LinearInterpolate(A=@n17, B=@n18, Alpha=@n20)), False=Constant()) {defaultValue=true, parameterName=UsePDO}
 ]
}
