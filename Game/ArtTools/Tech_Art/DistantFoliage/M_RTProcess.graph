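# Unlit translucent material that post-processes the distant foliage capture render target, dilating its color and coverage.
# It samples the center and eight neighbors at small BlurSampleOffsets UV offsets, clamped at the edges.
# Chained Max nodes keep the brightest RGB per channel, while chained Mins take the lowest alpha, which is then inverted.
# The Passthrough scalar blends both results toward the unfiltered center sample, with alpha still inverted.
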Material /Game/ArtTools/Tech_Art/DistantFoliage/M_RTProcess
graph {
 properties {bAutomaticallySetUsageInEditor=false, blendMode=BLEND_Translucent, shadingModel=MSM_Unlit, usedShadingModels=Unlit}
 nodes [
  n1 = TextureCoordinate()
  n2 = TextureSample(UVs=@n1) {samplerSource=SSM_Clamp_WorldGroupSettings, texture=&/RT_DistantFoliage_Capture}
  n3 = MaterialFunctionCall!(OffsetAmount=Constant() {r=0.0002}) {materialFunction=/Engine/Functions/Engine_MaterialFunctions01/ImageAdjustment/BlurSampleOffsets}
  n4 = TextureSample(UVs=Add(A=@n1, B=@n3."0,1")) {samplerSource=SSM_Clamp_WorldGroupSettings, texture=&RT_DistantFoliage_Capture}
  n5 = TextureSample(UVs=Add(A=@n1, B=@n3."1,0")) {samplerSource=SSM_Clamp_WorldGroupSettings, texture=&RT_DistantFoliage_Capture}
  n6 = TextureSample(UVs=Add(A=@n1, B=@n3."0,-1")) {samplerSource=SSM_Clamp_WorldGroupSettings, texture=&RT_DistantFoliage_Capture}
  n7 = Max(A=Max(A=Max(A=@n2.RGB, B=@n4.RGB), B=@n5.RGB), B=@n6.RGB)
  n8 = TextureSample(UVs=Add(A=@n1, B=@n3."-1,0")) {samplerSource=SSM_Clamp_WorldGroupSettings, texture=&RT_DistantFoliage_Capture}
  n9 = TextureSample(UVs=Add(A=@n1, B=@n3."-.5,.5")) {samplerSource=SSM_Clamp_WorldGroupSettings, texture=&RT_DistantFoliage_Capture}
  n10 = TextureSample(UVs=Add(A=@n1, B=@n3.".5,.5")) {samplerSource=SSM_Clamp_WorldGroupSettings, texture=&RT_DistantFoliage_Capture}
  n11 = Max(A=Max(A=Max(A=@n7, B=@n8.RGB), B=@n9.RGB), B=@n10.RGB)
  n12 = TextureSample(UVs=Add(A=@n1, B=@n3.".5,-.5")) {samplerSource=SSM_Clamp_WorldGroupSettings, texture=&RT_DistantFoliage_Capture}
  n13 = TextureSample(UVs=Add(A=@n1, B=@n3."-.5,-.5")) {samplerSource=SSM_Clamp_WorldGroupSettings, texture=&RT_DistantFoliage_Capture}
  n14 = ScalarParameter() {parameterName=Passthrough}
  LinearInterpolate(A=Max(A=Max(A=@n11, B=@n12.RGB), B=@n13.RGB), B=@n2.RGB, Alpha=@n14)
  n15 = Min(A=@n2.A, B=@n4.A)
  n16 = Min(A=Min(A=Min(A=@n15, B=@n5.A), B=@n6.A), B=@n8.A)
  n17 = Min(A=Min(A=Min(A=@n16, B=@n9.A), B=@n10.A), B=@n12.A)
  LinearInterpolate(A=OneMinus(None=Min(A=@n17, B=@n13.A)), B=OneMinus(None=@n2.A), Alpha=@n14)
 ]
}
