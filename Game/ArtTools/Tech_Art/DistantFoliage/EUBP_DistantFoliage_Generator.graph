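# Editor utility that bakes BP_MeshGrass foliage into per-cell distant foliage textures and meshes.
# It hides foliage by material filters, forces LOD and removes cull distances, then captures each grid cell orthographically, optionally multisampled with jitter, before restoring the actors.
# PCG-generated vertex data is Delaunay-triangulated and pruned into Geometry Script meshes, saved as static mesh assets or previewed as dynamic components.
# A second PCG component optionally spawns actors.
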EditorUtilityBlueprint /Game/ArtTools/Tech_Art/DistantFoliage/EUBP_DistantFoliage_Generator
graphs [
 CacheRenderActorData {nodes [
   n1 = CacheRenderActorData!() -> (then, SourceActor:Actor Object Reference) {is=title K2Node_FunctionEntry}
   n2 = Get Components By Class!(self:Actor Object Reference=@n1, ComponentClass:Actor Component Class Reference=/Script/Engine.InstancedStaticMeshComponent) -> (ReturnValue:Array of Instanced Static Mesh Component Object References)
   > n3 = Branch(Condition=Is Not Empty!(TargetArray:Array of Instanced Static Mesh Component Object References=@n2) -> (ReturnValue:bool))
   n4 = Length(TargetArray:Array of Instanced Static Mesh Component Object References=@n2)
   > Resize!(execute, TargetArray:Array of Int Point Structures=Get CullDistances() -> (CullDistances:Array of Int Point Structures), Size:int=@n4) -> (then) {is=title K2Node_CallArrayFunction}
   > Resize(TargetArray:Array of Booleans=Get Visibility() -> (Visibility:Array of Booleans), Size=@n4)
   > n5 = For Each Loop(Array:Array of Instanced Static Mesh Component Object References=@n2) -> (Array Element:Instanced Static Mesh Component Object Reference)
   Return Node!(execute=[@n3.else, @n5.Completed], ReturnValue:Array of Int Point Structures=Get CullDistances(), Vis:Array of Booleans=Get Visibility())
   > n6 = DetermineVisibility!(execute, self, ISM:Mesh Component Object Reference=@n5, UseExcludeList:bool=Get UseExcludeList() -> (UseExcludeList:bool)) -> (then, Visible:bool)
   > Set Array Elem!(execute, TargetArray:Array of Booleans=Get Visibility(), Index:int=@n5.Array Index, Item:"Boolean (by ref)"=@n6, bSizeToFit:bool=false) -> (then)
   > SetVisibility!(execute, self:Scene Component Object Reference=@n5, bNewVisibility:bool=@n6, bPropagateToChildren:bool=false) -> (then)
   n7 = GetCullDistances!(self:Instanced Static Mesh Component Object Reference=@n5) -> (OutStartCullDistance:int=0, OutEndCullDistance:int=0)
   > Set Array Elem!(
    execute
    TargetArray:Array of Int Point Structures=Get CullDistances()
    Index=@n5.Array Index
    Item_X:int=@n7
    Item_Y:int=@n7.OutEndCullDistance
    bSizeToFit=false
   ) -> (then)
   > SetCullDistances!(execute, self:Instanced Static Mesh Component Object Reference=@n5, StartCullDistance:int=0, EndCullDistance:int=0) -> (then)
   > SetForcedLodModel!(execute, self:Static Mesh Component Object Reference=@n5, NewForcedLodModel:int=1) -> (then)
  ]}
 CalculateMeshData {nodes [
   CalculateMeshData() {is=title K2Node_FunctionEntry}
   n1 = Get OutputTextures() -> (OutputTextures:Array of Texture 2D Object References)
   > n2 = Branch(Condition=Is Not Empty(TargetArray:Array of Texture 2D Object References=@n1))
   > n3 = Is Valid(InputObject="Get (a copy)"(Array:Array of Texture 2D Object References=@n1) -> (Output:Texture 2D Object Reference))
   n4 = Get OutputTextureAssets() -> (OutputTextureAssets:bool)
   PrintString(
    execute=[@n2.else, @n3.Is Not Valid]
    InString=Append(A=SelectString(A="Error: ", B="Warning: ", bPickA=@n4), B=CalculateMeshData: Aborting due to lack of textures)
   )
   n5 = Branch(execute=@n3.Is Valid, Condition=Get Debug() -> (Debug:bool))
   > n6 = LogString(InString=Calculating MeshData)
   GenerateLocal!(
    execute=[@n6.then, @n5.else]
    self:PCGComponent Object Reference=Get PCG_Generator() -> (PCG_Generator:PCGComponent Object Reference)
    bForce:bool=true
   ) -> (then)
  ]}
 DetermineVisibility {nodes [
   n1 = DetermineVisibility!() -> (then, ISM:Mesh Component Object Reference, UseExcludeList:bool=~) {is=title K2Node_FunctionEntry}
   > n2 = Branch(Condition=@n1.UseExcludeList)
   > n3 = Branch(Condition=Is Empty!(TargetArray:Array of Material Interface Object References=Get ExcludedMaterials_Internal() -> (ExcludedMaterials_Internal:Array of Material Interface Object References)) -> (ReturnValue:bool))
   > Return Node!(execute, Visible:bool=true)
   n4 = GetMaterials!(execute=@n3.else, self:Mesh Component Object Reference=@n1) -> (then, ReturnValue:Array of Material Interface Object References)
   > n5 = For Each Loop(Array:Array of Material Interface Object References=@n4) -> (Array Element:Material Interface Object Reference)
   Return Node(execute=@n5.Completed, Visible=true)
   n6 = Get ExcludedMaterials_Internal()
   > Branch(Condition="Not Equal (Integer)"(
     A=Find Item!(TargetArray:Array of Material Interface Object References=@n6, ItemToFind:"Material Interface Object Reference (by ref)"=@n5) -> (ReturnValue)
     B=-1
    ))
   > Return Node(Visible=false)
   n7 = Branch(
    execute=@n2.else
    Condition=Is Empty(TargetArray=Get IncludedMaterials_Internal() -> (IncludedMaterials_Internal:Array of Material Interface Object References))
   )
   > Return Node(Visible=true)
   n8 = GetMaterials(execute=@n7.else, self=@n1)
   > n9 = For Each Loop(Array=@n8)
   Return Node(execute=@n9.Completed, Visible=false)
   n10 = Get IncludedMaterials_Internal()
   > Branch(Condition="Not Equal (Integer)"(A=Find Item(TargetArray=@n10, ItemToFind=@n9), B=-1))
   > Return Node(Visible=true)
  ]}
 EventGraph {nodes [
   Resize Textures() {is=title K2Node_CustomEvent}
   > n1 = For Each Loop(Array:Array of Texture 2D Object References=Get OutputTextures()) -> (Array Element:Texture 2D Object Reference)
   > Set CompressionNoAlpha!(execute, CompressionNoAlpha:bool=true, self:Texture Object Reference=@n1) -> (then, Output_Get:bool)
   > Set MaxTextureSize!(execute, MaxTextureSize:int=Get TextureResolution() -> (TextureResolution:int=0), self:Texture Object Reference=@n1) -> (then, Output_Get:int=0)
   n2 = Get AssetEditorSubsystem()
   > OpenEditorForAssets!(
    execute
    self:Asset Editor Subsystem Object Reference=@n2
    Assets:Array of Object References=Make Array("[0]":Object Reference=@n1) -> (Array:Array of Object References)
    OpenedMethod:EAssetTypeActivationOpenedMethod Enum=View
   ) -> (then, ReturnValue:bool)
   > CloseAllEditorsForAsset!(execute, self:Asset Editor Subsystem Object Reference=@n2, Asset:Object Reference=@n1) -> (then, ReturnValue:int=0)
   "Generate!"() {is=title K2Node_CustomEvent}
   > GenerateTextures()
   > CalculateMeshData() {is=title K2Node_CallFunction}
   PCGReceive!() -> (OutputDelegate, then, PCGData:PCGData Collection Structure) {is=title K2Node_CustomEvent}
   > GenerateMeshes()
   > SpawnActors()
  ]}
 GatherActorsToRender {nodes [
   GatherActorsToRender() {is=title K2Node_FunctionEntry}
   > Clear(TargetArray:Array of Distant Foliage Render Actor Data Structures=Get RenderActorData() -> (RenderActorData:Array of Distant Foliage Render Actor Data Structures))
   > n1 = GetAllActorsOfClass!(
    execute
    WorldContextObject:Object Reference
    ActorClass:Actor Class Reference=/Game/Athena/Helios/Environments/Landscape/WorldBuildingTools/PCGGrass/BP_MeshGrass.BP_MeshGrass_C
   ) -> (then, OutActors:Array of BP Mesh Grass Object References)
   > n2 = Set RenderedActors(RenderedActors:Array of BP Mesh Grass Object References=@n1) -> (Output_Get:Array of BP Mesh Grass Object References)
   > n3 = For Each Loop(Array:Array of BP Mesh Grass Object References=@n2) -> (Array Element:BP Mesh Grass Object Reference)
   n4 = Set ShowOnlyActors!(
    execute=@n3.Completed
    ShowOnlyActors:Array of Actor Object References=Get RenderActors() -> (RenderActors:Array of Actor Object References)
    self:Scene Capture Component Object Reference=Get SceneCaptureComponent2D() -> (SceneCaptureComponent2D:Scene Capture Component 2D Object Reference)
   ) -> (then, Output_Get:Array of Actor Object References)
   > Return Node!(execute, Success:bool=Is Not Empty(TargetArray:Array of Actor Object References=@n4))
   Set RenderActor(execute=@n3.LoopBody, RenderActor:BP Mesh Grass Object Reference=@n3) -> (Output_Get:BP Mesh Grass Object Reference)
   n5 = Get RenderActor() -> (RenderActor:BP Mesh Grass Object Reference)
   > n6 = CacheRenderActorData!(execute, self, SourceActor:Actor Object Reference=@n5) -> (then, ReturnValue:Array of Int Point Structures, Vis:Array of Booleans) {is=title K2Node_CallFunction}
   > Add!(
    execute
    TargetArray:Array of Distant Foliage Render Actor Data Structures=Get RenderActorData()
    NewItem_Actor_2_#1:Actor Object Reference=@n5
    NewItem_CullDistances_7_#2:Array of Int Point Structures=@n6
    NewItem_Visiblity_10_#3:Array of Booleans=@n6.Vis
   ) -> (then, ReturnValue) {is=title K2Node_CallArrayFunction}
   n7 = Get RenderActor()
   n8 = vector + vector(
    A=GetScaledBoxExtent!(self:Box Collision Object Reference=Get Box() -> (Box:Box Collision Object Reference)) -> (ReturnValue:Vector)
    B=Get World Location!(self:Scene Component Object Reference=Get Box()) -> (ReturnValue:Vector) {is=title K2Node_CallFunction}
   )
   > Branch(Condition=TestPointInsideBox!(
     Box_Min:Vector=NegateVector!(A:Vector=@n8) -> (ReturnValue:Vector)
     Box_Max:Vector=@n8
     Point:Vector=Get Actor Location(self=@n7)
     bConsiderOnBoxAsInside:bool=true
    ) -> (ReturnValue:bool))
   > Add!(execute, TargetArray:Array of Actor Object References=Get RenderActors(), NewItem:"Actor Object Reference (by ref)"=Get RenderActor()) -> (then, ReturnValue)
   > Branch(Condition=AND Boolean(A=Get Debug(), B=Get Verbose() -> (Verbose:bool)))
   > LogString(InString=GetDisplayName(Object=Get RenderActor()))
  ]}
 GenerateMeshes {nodes [
   Clear(TargetArray:Array of Vectors=Get 3DPositionOrdered() -> (3DPositionOrdered:Array of Vectors))
   > n1 = Clear(TargetArray:Array of Integers=Get InvalidIndices_Ordered() -> (InvalidIndices_Ordered:Array of Integers))
   n2 = AppendDelaunayTriangulation2D!(
    execute=@n19.then
    TargetMesh:Dynamic Mesh Object Reference=Get CurrentMesh() -> (CurrentMesh:Dynamic Mesh Object Reference)
    PrimitiveOptions:Geometry Script Primitive Options Structure=Make Geometry Script Primitive Options!(PolygroupMode:EGeometryScriptPrimitivePolygroupMode Enum=SingleGroup) -> (GeometryScriptPrimitiveOptions:Geometry Script Primitive Options Structure) {is=title K2Node_MakeStruct}
    Transform:Transform
    VertexPositions:Array of Vector 2D Structures=Get Positions2D() -> (Positions2D:Array of Vector 2D Structures)
    ConstrainedEdges:Array of Int Point Structures
    TriangulationOptions:Geometry Script Constrained Delaunay Triangulation Options Structure=Make Geometry Script Constrained Delaunay Triangulation Options!() -> (GeometryScriptConstrainedDelaunayTriangulationOptions:Geometry Script Constrained Delaunay Triangulation Options Structure) {is=title K2Node_MakeStruct}
   ) -> (then, PositionsToVertexIDs:Array of Integers, bHasDuplicateVertices:bool, ReturnValue:Dynamic Mesh Object Reference)
   n3 = For Each Loop(Exec=@n1.then, Array:Array of Integers=@n2) -> (Array Element:int)
   Set Positions3D(execute=@n3.Completed, Positions3D:Array of Vectors=Get 3DPositionOrdered()) -> (Output_Get:Array of Vectors)
   > n4 = Set InvalidIndices(InvalidIndices:Array of Integers=Get InvalidIndices_Ordered()) -> (Output_Get:Array of Integers)
   n5 = ConvertArrayToVectorList!(execute=[@n2.then, @n4.then], VectorArray:Array of Vectors=Get Positions3D() -> (Positions3D:Array of Vectors)) -> (then, VectorList:Vector List Structure)
   > SetAllMeshVertexPositions!(execute, TargetMesh:Dynamic Mesh Object Reference=@n2.ReturnValue, PositionList:Vector List Structure=@n5) -> (then, ReturnValue:Dynamic Mesh Object Reference)
   > n6 = ConvertArrayToIndexList!(
    execute
    IndexArray:Array of Integers=Get InvalidIndices() -> (InvalidIndices:Array of Integers)
    IndexType:EGeometryScriptIndexType Enum=Vertex
   ) -> (then, IndexList:Index List Structure)
   > DeleteVerticesFromMesh!(
    execute
    TargetMesh:Dynamic Mesh Object Reference=Get CurrentMesh()
    VertexList:Index List Structure=@n6
    bDeferChangeNotifications:bool=false
   ) -> (then, NumDeleted:int=0, ReturnValue:Dynamic Mesh Object Reference)
   n7 = GetVertexCount!(TargetMesh:Dynamic Mesh Object Reference=Get CurrentMesh()) -> (ReturnValue:int=0)
   > Add(TargetArray:Array of Integers=Get VertCounts() -> (VertCounts:Array of Integers), NewItem:"Integer (by ref)"=@n7)
   > n8 = Set IsMeshValid(IsMeshValid:bool="integer > integer"(A=@n7)) -> (Output_Get:bool)
   > Branch(Condition=Get Debug())
   > LogString(InString=Append(A="IsMeshValid: ", B="To String (Boolean)"(InBool=@n8)))
   Add(
    execute=@n3.LoopBody
    TargetArray:Array of Vectors=Get 3DPositionOrdered()
    NewItem:"Vector (by ref)"="Get (a copy)"(Array:Array of Vectors=Get Positions3D(), Dimension 1=@n3) -> (Output:Vector)
   )
   > Add(
    TargetArray:Array of Integers=Get InvalidIndices_Ordered()
    NewItem:"Integer (by ref)"="Get (a ref)"(Array=Get InvalidIndices(), Dimension 1=@n3)
   )
   GenerateMeshes() {is=title K2Node_FunctionEntry}
   n9 = Get MeshData() -> (MeshData:Array of Distant Foliage Mesh Data Structures)
   > n10 = Branch(Condition=Is Not Empty(TargetArray:Array of Distant Foliage Mesh Data Structures=@n9))
   > n11 = Branch(Condition=Get Debug())
   n12 = Length(TargetArray:Array of Distant Foliage Mesh Data Structures=@n9)
   n13 = Length(TargetArray:Array of Vectors="Get (a copy)"!(Array:Array of Distant Foliage Mesh Data Structures=@n9, Dimension 1) -> (
     Output_Position3D_17_#4:Array of Vectors
     Output_Position2D_18_#5:Array of Vector 2D Structures
     Output_InvalidIndices_19_#6:Array of Integers
    ))
   > n14 = LogString(InString=Append!(A=Num MeshData:, B="To String (Integer)"(InInt=@n12), C:String=" - ", D:String="To String (Integer)"(InInt=@n13)) -> (ReturnValue))
   n15 = For Each Loop!(Exec=[@n14.then, @n11.else], Array:Array of Distant Foliage Mesh Data Structures=Get MeshData()) -> (
    LoopBody
    Array Element_Position3D_17_#4:Array of Vectors
    Array Element_Position2D_18_#5:Array of Vector 2D Structures
    Array Element_InvalidIndices_19_#6:Array of Integers
    Array Index
    Completed
   )
   Clear(execute=@n15.Completed, TargetArray:Array of Distant Foliage Mesh Data Structures=Get MeshData())
   Set Positions3D(execute=@n15.LoopBody, Positions3D=@n15)
   > Set Positions2D(Positions2D:Array of Vector 2D Structures=@n15.Array Element_Position2D_18_#5) -> (Output_Get:Array of Vector 2D Structures)
   > Set InvalidIndices(InvalidIndices=@n15.Array Element_InvalidIndices_19_#6)
   > Set CurrentIndex(CurrentIndex:int=@n15.Array Index) -> (Output_Get:int=0)
   > n16 = Sequence()
   > n17 = Construct Object from Class(Class=/Script/GeometryFramework.DynamicMesh) -> (ReturnValue:Dynamic Mesh Object Reference)
   > n18 = Set CurrentMesh(CurrentMesh:Dynamic Mesh Object Reference=@n17) -> (Output_Get:Dynamic Mesh Object Reference)
   > n19 = Add(
    TargetArray:Array of Dynamic Mesh Object References=Get DynamicMeshes() -> (DynamicMeshes:Array of Dynamic Mesh Object References)
    NewItem:"Dynamic Mesh Object Reference (by ref)"=@n18
   )
   n20 = Branch(execute=@n16.then_1, Condition=Get OutputMeshAssets() -> (OutputMeshAssets:bool))
   > Branch(Condition=Get IsMeshValid() -> (IsMeshValid:bool))
   n21 = Get AssetNameSuffixes() -> (AssetNameSuffixes:Array of Strings)
   n22 = Get CurrentIndex() -> (CurrentIndex:int=0)
   > n23 = CreateNewStaticMeshAssetFromMesh!(
    execute
    FromDynamicMesh:Dynamic Mesh Object Reference=Get CurrentMesh()
    AssetPathAndName:String=Append!(
     A=GetSanitizedOutputPath!() -> (ReturnValue:String)
     B=SM_DistantFoliage
     C="Get (a copy)"!(Array:Array of Strings=@n21, Dimension 1=@n22) -> (Output:String)
    ) -> (ReturnValue)
    Options:Geometry Script Create New Static Mesh Asset Options Structure=Make Geometry Script Create New Static Mesh Asset Options!() -> (GeometryScriptCreateNewStaticMeshAssetOptions:Geometry Script Create New Static Mesh Asset Options Structure) {is=title K2Node_MakeStruct}
   ) -> (Failure:Exec, Success:Exec, ReturnValue:Static Mesh Object Reference)
   n24 = Get StaticMeshEditorSubsystem()
   EnableSectionCollision!(
    execute=@n23.Success
    self:Static Mesh Editor Subsystem Object Reference=@n24
    StaticMesh:Static Mesh Object Reference=@n23
    bCollisionEnabled:bool=false
    LODIndex:int=0
    SectionIndex:int=0
   ) -> (then)
   > EnableSectionVisibleInRayTracing!(
    execute
    self:Static Mesh Editor Subsystem Object Reference=@n24
    StaticMesh:Static Mesh Object Reference=@n23
    bVisibleInRayTracing:bool=false
    LODIndex:int=0
    SectionIndex:int=0
   ) -> (then)
   > EnableSectionAffectDistanceFieldLighting!(
    execute
    self:Static Mesh Editor Subsystem Object Reference=@n24
    StaticMesh:Static Mesh Object Reference=@n23
    bAffectDistanceFieldLighting:bool=false
    LODIndex:int=0
    SectionIndex:int=0
   ) -> (then)
   > EnableSectionCastShadow!(
    execute
    self:Static Mesh Editor Subsystem Object Reference=@n24
    StaticMesh:Static Mesh Object Reference=@n23
    bCastShadow:bool=false
    LODIndex:int=0
    SectionIndex:int=0
   ) -> (then)
   > Set Array Elem!(
    execute
    TargetArray:Array of Static Mesh Object References=Get OutputMeshes() -> (OutputMeshes:Array of Static Mesh Object References)
    Index=Get CurrentIndex() -> (CurrentIndex=0)
    Item:"Static Mesh Object Reference (by ref)"=@n23
    bSizeToFit=false
   ) -> (then)
   > SetMaterial!(
    execute
    self:Static Mesh Object Reference=@n23
    MaterialIndex:int=0
    NewMaterial:Material Interface Object Reference=&/M_DistantFoliage
   ) -> (then)
   > n25 = Branch(Condition=IsUnattended!() -> (ReturnValue:bool))
   > n26 = AddAssetToSubmitAllowList!(
    execute
    InPath:"Soft Object Reference (by ref)"=Make Soft Reference!(Input:Static Mesh Object Reference=@n23) -> (Output:Static Mesh Soft Object Reference)
   ) -> (then)
   Branch(execute=[@n26.then, @n25.else], Condition=Get Debug())
   > Add(
    TargetArray:Array of Static Mesh Object References=Get GeneratedMeshes() -> (GeneratedMeshes:Array of Static Mesh Object References)
    NewItem:"Static Mesh Object Reference (by ref)"=@n23
   )
   n27 = Add Dynamic Mesh Component!(
    execute:Exec=@n20.else
    self:Actor Object Reference
    bManualAttachment:bool=false
    RelativeTransform:Transform
    ComponentTemplateContext:Object Reference
   ) -> (then:Exec, ReturnValue:Dynamic Mesh Component Object Reference) {is=title K2Node_AddComponent}
   > Set bExplicitShowWireframe!(execute, bExplicitShowWireframe:bool=Get ShowWireframe() -> (ShowWireframe:bool), self:Base Dynamic Mesh Component Object Reference=@n27) -> (then, Output_Get:bool)
   n28 = "Get (a ref)"!(Array:Array of Vectors=Get BucketCoordinates() -> (BucketCoordinates:Array of Vectors), Dimension 1=Get CurrentIndex() -> (CurrentIndex=0)) -> (Output_X:double, Output_Y:double, Output_Z:double)
   > Set World Transform!(
    execute
    self:Scene Component Object Reference=@n27
    NewTransform_Location:Vector=MakeVector(X=@n28, Y=@n28.Output_Y, Z=Get MeshBiasZ() -> (MeshBiasZ:double))
    NewTransform_Rotation:Rotator="0, 0, 0"
    NewTransform_Scale:Vector="1,1,1"
    bSweep:bool=false
    bTeleport:bool=false
   ) -> (then, SweepHitResult:Hit Result Structure) {is=title K2Node_CallFunction}
   > SetDynamicMesh!(execute, self:Dynamic Mesh Component Object Reference=@n27, NewMesh:Dynamic Mesh Object Reference=Get CurrentMesh()) -> (then)
   > n29 = CreateDynamicMaterialInstance!(
    execute
    self:Primitive Component Object Reference=@n27
    ElementIndex:int=0
    SourceMaterial:Material Interface Object Reference=&M_DistantFoliage
    OptionalName:Name=None
   ) -> (then, ReturnValue:Material Instance Dynamic Object Reference) {is=title K2Node_CallFunction}
   > SetTextureParameterValue(
    self=@n29
    ParameterName=FoliageTexture
    Value="Get (a copy)"(Array:Array of Texture 2D Object References=Get OutputTextures(), Dimension 1=Get CurrentIndex() -> (CurrentIndex=0)) -> (Output:Texture 2D Object Reference)
   )
   > SetAffectDynamicIndirectLighting!(execute, self:Primitive Component Object Reference=@n27, bNewAffectDynamicIndirectLighting:bool=false) -> (then)
   > SetAffectDistanceFieldLighting!(execute, self:Primitive Component Object Reference=@n27, NewAffectDistanceFieldLighting:bool=false) -> (then)
   PrintString(execute=@n10.else, InString=Error: Aborting due to lack of MeshData)
  ]}
 GenerateTextures {nodes [
   GenerateTextures() {is=title K2Node_FunctionEntry}
   > ExecuteConsoleCommand(Command=r.DetailMode 3)
   > n1 = GatherActorsToRender!(execute, self) -> (then, Success:bool) {is=title K2Node_CallFunction}
   > Branch(Condition=@n1)
   > n2 = Sequence!(execute) -> (then_0, then_1, then_2:Exec, then_3:Exec)
   > n3 = Sequence!(execute) -> (then_0, then_1)
   > Clear(TargetArray:Array of Strings=Get AssetNameSuffixes())
   > Clear(TargetArray:Array of Static Mesh Object References=Get OutputMeshes())
   > Clear(TargetArray:Array of Texture 2D Object References=Get OutputTextures())
   > Clear(TargetArray=Get GeneratedTextures() -> (GeneratedTextures:Array of Texture 2D Object References))
   > Clear(TargetArray:Array of Static Mesh Object References=Get GeneratedMeshes())
   > Clear(TargetArray:Array of Integers=Get VertCounts())
   > Clear(TargetArray:Array of Dynamic Mesh Object References=Get DynamicMeshes())
   > n4 = Set RT_Capture(RT_Capture:Texture Render Target 2D Object Reference=Get TextureTarget!(self:Scene Capture Component 2D Object Reference=Get SceneCaptureComponent2D()) -> (TextureTarget:Texture Render Target 2D Object Reference)) -> (Output_Get:Texture Render Target 2D Object Reference)
   n5 = Get RenderResolution() -> (RenderResolution:int=0)
   > ResizeRenderTarget2D!(execute, TextureRenderTarget:Texture Render Target 2D Object Reference=@n4, Width:int=@n5, Height:int=@n5) -> (then)
   > n6 = Set RT_Output(RT_Output:Texture Render Target 2D Object Reference=&/RT_DistantFoliage_Output) -> (Output_Get:Texture Render Target 2D Object Reference)
   n7 = Get TextureResolution() -> (TextureResolution=0)
   > ResizeRenderTarget2D(TextureRenderTarget=@n6, Width=@n7, Height=@n7)
   > n8 = Get Components By Class!(execute:Exec, self, ComponentClass=/Script/GeometryFramework.DynamicMeshComponent) -> (then:Exec, ReturnValue:Array of Dynamic Mesh Component Object References)
   > n9 = For Each Loop!(Exec, Array:Array of Dynamic Mesh Component Object References=@n8) -> (LoopBody, Array Element:Dynamic Mesh Component Object Reference, Array Index, Completed)
   > Destroy Component!(execute, self:Actor Component Object Reference=@n9, Object:Object Reference) -> (then)
   n10 = Truncate(A=Get TotalBuckets() -> (TotalBuckets:double))
   Branch(execute=@n3.then_1, Condition="integer > integer"(A=@n10))
   > n11 = Branch(Condition=Get UseSelectiveBuckets() -> (UseSelectiveBuckets:bool))
   n12 = Get SelectiveBuckets() -> (SelectiveBuckets:Array of Integers)
   > n13 = Branch(Condition=Is Not Empty(TargetArray:Array of Integers=@n12))
   > n14 = For Each Loop(Array:Array of Integers=@n12) -> (Array Element:int)
   n15 = Set NumActiveBuckets(
    execute=@n14.Completed
    NumActiveBuckets:int=Length(TargetArray:Array of Integers=Get SelectiveBucketsUnique() -> (SelectiveBucketsUnique:Array of Integers))
   ) -> (Output_Get:int=0)
   n16 = Sequence(execute=[@n15.then, @n67.then])
   n17 = Get NumActiveBuckets() -> (NumActiveBuckets:int=0)
   > Resize(TargetArray:Array of Texture 2D Object References=Get OutputTextures(), Size=@n17)
   > Resize(TargetArray:Array of Static Mesh Object References=Get OutputMeshes(), Size=@n17)
   > Resize(TargetArray:Array of Strings=Get AssetNameSuffixes(), Size=@n17)
   > n18 = GetCurrentLevelName!(execute, WorldContextObject:Object Reference, bRemovePrefixString:bool=true) -> (then, ReturnValue:String)
   > Set LevelName(LevelName:String=@n18) -> (Output_Get:String)
   n19 = AND Boolean(A=Get Debug(), B=Get UseDebugRange() -> (UseDebugRange:bool))
   n20 = int - int(A=@n10)
   n21 = Get!() -> (DebugRange_X:int, DebugRange_Y:int)
   n22 = SelectInt(A="Min (Integer)"(A=@n21.DebugRange_Y, B=@n20) -> (ReturnValue=0), B=@n20, bPickA=@n19)
   n23 = SelectInt(A="Max (Integer)"(A=@n21, B=0), B=0, bPickA=@n19)
   n24 = For Loop(execute=@n16.then_1, FirstIndex=@n23, LastIndex=@n22)
   > n25 = Branch(Condition=Get UseSelectiveBuckets())
   > n26 = Branch(Condition=Contains Item(TargetArray:Array of Integers=Get SelectiveBucketsUnique(), ItemToFind:"Integer (by ref)"=@n24))
   Increment Int(" "=[@n25.else, @n26.then], Value=Get IterationIndex() -> (IterationIndex:int=0))
   > n27 = Set BucketIndex(BucketIndex:int=@n24) -> (Output_Get:int=0)
   > Set CellIndex(CellIndex:double="To Float (Integer)"(InInt=@n27)) -> (Output_Get:double)
   n28 = Get CellIndex() -> (CellIndex:double)
   n29 = Get NumBucketsFloat() -> (NumBucketsFloat:double)
   n30 = "% (Float)"!(A:double=@n28, B:double=@n29) -> (ReturnValue:double)
   > Set CellColumn(CellColumn:int=Truncate(A=@n30)) -> (Output_Get:int=0)
   n31 = Floor(A=float / float(A=@n28, B=@n29)) {is=title K2Node_CallFunction}
   > Set CellRow(CellRow:int=@n31) -> (Output_Get:int=0)
   n32 = BreakVector2D(InVec=Get GridOrigin() -> (GridOrigin:Vector 2D Structure))
   n33 = Get CellExtent() -> (CellExtent:double)
   n34 = float + float(A=@n33, B=float * float(A=Get CellSize() -> (CellSize:double), B=@n30) {is=title K2Node_PromotableOperator})
   n35 = float + float(A=@n33, B=float * float(A="To Float (Integer)"(InInt=@n31), B=Get CellSize()))
   > Set CellPosition(CellPosition:Vector=MakeVector(X=float + float(A=@n32, B=@n34), Y=float + float(A=@n32.Y, B=@n35), Z=Get BoxTop() -> (BoxTop:double))) -> (Output_Get:Vector)
   > n36 = Branch(Condition=Get Debug())
   n37 = Get TotalBuckets()
   n38 = Get CellIndex()
   > n39 = DrawDebugBox!(
    execute
    WorldContextObject:Object Reference
    Center:Vector=Get CellPosition() -> (CellPosition:Vector)
    Extent:Vector="To Vector (Float)"!(InDouble:double=Get CellExtent()) -> (ReturnValue:Vector)
    LineColor:Linear Color Structure="To LinearColor (Float)"!(InDouble:double=float / float(A=@n38, B=@n37)) -> (ReturnValue:Linear Color Structure)
    Rotation:Rotator
    Duration:float=5
    Thickness:float=1000
    DepthPriority:EDrawDebugSceneDepthPriorityGroup Enum=World
   ) -> (then)
   n40 = Get CellColumn() -> (CellColumn:int=0)
   n41 = Get CellRow() -> (CellRow:int=0)
   n42 = Set AssetNameSuffix(
    execute=[@n39.then, @n36.else]
    AssetNameSuffix:String=Append!(
     A=_
     B=Get LevelName() -> (LevelName:String)
     C=_
     D:String="To String (Integer)"(InInt=@n41)
     E:String=_
     F:String="To String (Integer)"(InInt=@n40)
    ) -> (ReturnValue)
   ) -> (Output_Get:String)
   > Set Array Elem(
    TargetArray:Array of Strings=Get AssetNameSuffixes()
    Index=Get IterationIndex() -> (IterationIndex=0)
    Item:"String (by ref)"=@n42
    bSizeToFit=false
   )
   > n43 = Branch(Condition=Get MultiSample() -> (MultiSample:bool))
   > ClearRenderTarget2D!(
    execute
    WorldContextObject:Object Reference
    TextureRenderTarget:Texture Render Target 2D Object Reference=Get RT_Capture() -> (RT_Capture:Texture Render Target 2D Object Reference)
    ClearColor:Linear Color Structure=(R=0,G=0,B=0,A=1)
   ) -> (then)
   > Set CompositeMode!(
    execute
    CompositeMode:ESceneCaptureCompositeMode Enum=SCCM_Composite
    self:Scene Capture Component 2D Object Reference=Get SceneCaptureComponent2D()
   ) -> (then, Output_Get:ESceneCaptureCompositeMode Enum=SCCM_Overwrite)
   n44 = Get MultiSampleCount() -> (MultiSampleCount:int=0)
   > n45 = For Loop(LastIndex=@n44)
   n46 = Self-Reference()
   n47 = CreateDynamicMaterialInstance(
    execute=[@n66.then, @n45.Completed]
    WorldContextObject=@n46
    Parent=&/M_RTProcess
    OptionalName=None
    CreationFlags=None
   ) {is=title K2Node_CallMaterialParameterCollectionFunction}
   n48 = Get PostProcessTextures() -> (PostProcessTextures:bool)
   > SetScalarParameterValue(self=@n47, ParameterName=Passthrough, Value="To Float (Boolean)"(InBool=NOT Boolean(A=@n48)))
   > DrawMaterialToRenderTarget!(
    execute
    WorldContextObject:Object Reference=@n46
    TextureRenderTarget:Texture Render Target 2D Object Reference=Get RT_Output() -> (RT_Output:Texture Render Target 2D Object Reference)
    Material:Material Interface Object Reference=@n47
   ) -> (then)
   > n49 = Set TargetTexture(TargetTexture:String=Append!(A=GetSanitizedOutputPath(), B=T_DistantFoliage, C=Get AssetNameSuffix() -> (AssetNameSuffix:String)) -> (ReturnValue)) -> (Output_Get:String)
   > n50 = DoesAssetExist(AssetPath=@n49)
   > n51 = Branch(Condition=@n50)
   > n52 = LoadAsset(AssetPath=Get TargetTexture() -> (TargetTexture:String))
   > n53 = Cast To Texture2D(Object=@n52)
   > n54 = Set CurrentTexture(CurrentTexture:Texture 2D Object Reference=@n53) -> (Output_Get:Texture 2D Object Reference)
   > Branch(Condition=Get OutputTextureAssets())
   n55 = Get RT_Output()
   > ConvertRenderTargetToTexture2DEditorOnly(RenderTarget=@n55, Texture=@n54)
   > n56 = Set CompressionSettings(CompressionSettings=TC_BC7, self=@n54)
   Set Array Elem(
    execute=[@n61.then, @n56.then]
    TargetArray:Array of Texture 2D Object References=Get OutputTextures()
    Index=Get IterationIndex() -> (IterationIndex=0)
    Item:"Texture 2D Object Reference (by ref)"=Get CurrentTexture() -> (CurrentTexture:Texture 2D Object Reference)
    bSizeToFit=false
   )
   > ClearRenderTarget2D(TextureRenderTarget=Get RT_Output(), ClearColor=(R=0,G=0,B=0,A=1))
   > n57 = Branch(Condition=IsUnattended())
   > n58 = AddAssetToSubmitAllowList(InPath=Make Soft Reference(Input:Texture 2D Object Reference=Get CurrentTexture()) -> (Output:Texture 2D Soft Object Reference))
   Branch(execute=[@n58.then, @n57.else], Condition=Get Debug())
   n59 = Get CurrentTexture()
   > LogString(InString=Append!(A="Output Texture: ", B=GetDisplayName(Object=@n59)) -> (ReturnValue))
   Branch(execute=@n51.else, Condition=Get OutputTextureAssets())
   > n60 = Render Target 2D Create Static Texture 2D Editor Only!(
    execute
    RenderTarget:Texture Render Target 2D Object Reference=@n55
    Name:String=Get TargetTexture()
    CompressionSettings:TextureCompressionSettings Enum=TC_BC7
    MipSettings:TextureMipGenSettings Enum=TMGS_FromTextureGroup
   ) -> (then, ReturnValue:Texture 2D Object Reference)
   > n61 = Set CurrentTexture(CurrentTexture=@n60)
   n62 = vector * vector(
    A="To Vector (Vector2D)"!(
     InVector2D:Vector 2D Structure=Random Vector 2d in Range!(Min:Vector 2D Structure=(X=-1,Y=-1), Max:Vector 2D Structure=(X=1,Y=1)) -> (Vector2dOut:Vector 2D Structure) {is=title K2Node_MacroInstance}
     Z:float=0
    ) -> (ReturnValue:Vector)
    B=Get MultiSampleJitter() -> (MultiSampleJitter:double)
   ) -> (ReturnValue=~)
   n63 = Get CellPosition()
   n64 = Get SceneCaptureComponent2D()
   Set World Location!(
    execute=@n45.LoopBody
    self:Scene Component Object Reference=@n64
    NewLocation:Vector=SelectVector(A=@n63, B=vector + vector(A=@n63, B=@n62), bPickA="Equal (Integer)"(A=@n45, B=@n44))
    bSweep:bool=false
    bTeleport:bool=true
   ) -> (then, SweepHitResult:Hit Result Structure) {is=title K2Node_CallFunction}
   > CaptureScene(self:Scene Capture Component 2D Object Reference=@n64)
   n65 = Get SceneCaptureComponent2D()
   Set CompositeMode(execute=@n43.else, CompositeMode=SCCM_Overwrite, self=@n65) -> (Output_Get=SCCM_Overwrite)
   > Set World Location(self=@n65, NewLocation=Get CellPosition(), bSweep=false, bTeleport=true)
   > n66 = CaptureScene(self=@n65)
   Add Unique(execute=@n14.LoopBody, TargetArray:Array of Integers=Get SelectiveBucketsUnique(), NewItem:"Integer (by ref)"=@n14)
   LogString(execute=@n13.else, InString=Error: Aborting due to no buckets selected)
   n67 = Set NumActiveBuckets(execute=@n11.else, NumActiveBuckets=int + int(A=int - int(A=@n22, B=@n23))) -> (Output_Get=0)
   RestoreRenderActorData(execute=@n2.then_1)
   > Set ShowOnlyActors(ShowOnlyActors=Make Array!() -> (Array:Array of Actor Object References), self=Get SceneCaptureComponent2D())
   Branch(execute=@n2.then_2, Condition=Get Debug())
   > Set GeneratedTextures(GeneratedTextures:Array of Texture 2D Object References=Get OutputTextures()) -> (Output_Get:Array of Texture 2D Object References)
   K2Node_Composite_0 = DebugComposition!() {is=title K2Node_Composite}
  ]}
 DebugComposition {
  path=GenerateTextures.K2Node_Composite_0.DebugComposition
  nodes [
   n1 = To Text!(Value:Date Time Structure=UTC Now!() -> (ReturnValue:Date Time Structure)) -> (ReturnValue:Text) {is=title K2Node_GenericToText}
   n2 = Replace(SourceString=Replace(SourceString="To String (Text)"(InText=@n1), From=" "), From=:, To=_)
   n3 = Render Target 2D Create Static Texture 2D Editor Only(
    RenderTarget=Get TextureTarget(self=Get SceneCaptureComponent2D())
    Name=Append!(A=Get TargetTexture(), B=_RT_, C:String=Replace(SourceString=@n2, From=",", To=_)) -> (ReturnValue)
    CompressionSettings=TC_BC7
    MipSettings=TMGS_FromTextureGroup
   )
   > Branch(Condition=IsUnattended())
   > AddAssetToSubmitAllowList(InPath=Make Soft Reference(Input=@n3))
   Inputs!()
   Outputs()
  ]
 }
 GetSanitizedOutputPath {nodes [
   GetSanitizedOutputPath() {is=title K2Node_FunctionEntry}
   n1 = Get OutputPath() -> (OutputPath:String)
   n2 = Get OutputPath()
   n3 = SelectString(
    A=@n2
    B=RightChop!(SourceString:String=@n2, Count:int=4) -> (ReturnValue:String)
    bPickA=NOT Boolean(A=StartsWith(SourceString=@n2, InPrefix=/All))
   )
   > Return Node!(execute, ReturnValue:String=SelectString(A=@n3, B=Append!(A=@n3, B=/) -> (ReturnValue), bPickA=EndsWith(SourceString=@n1, InSuffix=/)))
  ]}
 RestoreRenderActorData {nodes [
   RestoreRenderActorData() {is=title K2Node_FunctionEntry}
   > n1 = For Each Loop!(Exec, Array:Array of Distant Foliage Render Actor Data Structures=Get RenderActorData()) -> (
    LoopBody
    Array Element_Actor_2_#1:Actor Object Reference
    Array Element_CullDistances_7_#2:Array of Int Point Structures
    Array Element_Visiblity_10_#3:Array of Booleans
    Array Index
    Completed
   )
   Clear(
    execute=@n1.Completed
    TargetArray:Array of BP Mesh Grass Object References=Get RenderedActors() -> (RenderedActors:Array of BP Mesh Grass Object References)
   )
   > Clear(TargetArray:Array of Distant Foliage Render Actor Data Structures=Get RenderActorData())
   n2 = For Each Loop!(
    Exec=@n1.LoopBody
    Array:Array of Instanced Static Mesh Component Object References=Get Components By Class!(self=@n1, ComponentClass=&InstancedStaticMeshComponent) -> (ReturnValue:Array of Instanced Static Mesh Component Object References)
   ) -> (LoopBody, Array Element:Instanced Static Mesh Component Object Reference, Array Index, Completed)
   n3 = "Get (a ref)"!(Array:Array of Int Point Structures=@n1.Array Element_CullDistances_7_#2, Dimension 1=@n2.Array Index) -> (Output_X:int, Output_Y:int)
   > SetCullDistances(self=@n2, StartCullDistance=@n3, EndCullDistance=@n3.Output_Y)
   > SetForcedLodModel(self=@n2, NewForcedLodModel=0)
   > SetVisibility(self=@n2, bNewVisibility=true, bPropagateToChildren=false)
  ]}
 SpawnActors {nodes [
   SpawnActors() {is=title K2Node_FunctionEntry}
   > Branch(Condition=AND Boolean(A=Get OutputActors() -> (OutputActors:bool), B=Get OutputMeshAssets()))
   > GenerateLocal(self=Get PCG_Spawner() -> (PCG_Spawner:PCGComponent Object Reference), bForce=true)
  ]}
 UserConstructionScript {nodes [
   Construction Script()
   > n1 = Branch(Condition=IsUnattended())
   > n2 = ExecuteConsoleCommand(Command=r.DetailMode 3)
   n3 = Sequence(execute=[@n2.then, @n1.else])
   > WarmUp()
   n4 = Sequence!(execute=@n3.then_1) -> (then_0, then_1, then_2:Exec, then_3:Exec, then_4:Exec)
   > Clear(TargetArray:Array of Material Interface Object References=Get IncludedMaterials_Internal())
   > n5 = For Each Loop(Array:Array of Material Interface Object References=Get IncludedMaterials() -> (IncludedMaterials:Array of Material Interface Object References)) -> (Array Element:Material Interface Object Reference)
   Clear(execute=@n5.Completed, TargetArray=Get ExcludedMaterials_Internal())
   > n6 = For Each Loop(Array=Get ExcludedMaterials() -> (ExcludedMaterials:Array of Material Interface Object References))
   n7 = Get ExcludedMaterials_Internal()
   > Add Unique(TargetArray:Array of Material Interface Object References=@n7, NewItem:"Material Interface Object Reference (by ref)"=@n6)
   > Add Unique(
    TargetArray=@n7
    NewItem=GetNaniteOverideMaterial!(self:Material Interface Object Reference=@n6) -> (ReturnValue:Material Interface Object Reference)
   )
   n8 = Get IncludedMaterials_Internal()
   Add Unique(execute=@n5.LoopBody, TargetArray=@n8, NewItem=@n5)
   > Add Unique(TargetArray=@n8, NewItem=GetNaniteOverideMaterial(self=@n5))
   n9 = Set NumBucketsFloat(execute=@n4.then_1, NumBucketsFloat:double="To Float (Integer)"(InInt=Get NumBuckets() -> (NumBuckets:int=0))) -> (Output_Get:double)
   > Set TotalBuckets(TotalBuckets:double=Power!(Base:double=@n9, Exp:double=2) -> (ReturnValue:double) {is=title K2Node_CallFunction}) -> (Output_Get:double)
   n10 = Get Box()
   > Set BoxTop(BoxTop:double=float + float!(
     A=GetScaledBoxExtent!(self=@n10) -> (ReturnValue_X:double, ReturnValue_Y:double, ReturnValue_Z:double).ReturnValue_Z
     B=Get Actor Location!(self) -> (ReturnValue_X:double, ReturnValue_Y:double, ReturnValue_Z:double).ReturnValue_Z
     C:double=Get CaptureHeightBias() -> (CaptureHeightBias:double)
    ) -> (ReturnValue)) -> (Output_Get:double)
   n11 = "To Vector2D (Vector)"!(InVector:Vector=GetScaledBoxExtent!(self=Get Box()) -> (ReturnValue:Vector)) -> (ReturnValue:Vector 2D Structure)
   n12 = Set GridSize(
    execute=@n4.then_2
    GridSize:double=float * float(A=GetMax2D!(A:Vector 2D Structure=@n11) -> (ReturnValue:double), B=Make Literal Float(Value=2))
   ) -> (Output_Get:double)
   > n13 = Set CellSize(CellSize:double=float / float(A=@n12, B=Get NumBucketsFloat())) -> (Output_Get:double)
   > Set CellExtent(CellExtent:double=float / float(A=@n13, B=2)) -> (Output_Get:double)
   n14 = Set OrthoWidth!(execute=@n4.then_3, OrthoWidth:float=Get CellSize(), self:Scene Capture Component 2D Object Reference=Get SceneCaptureComponent2D()) -> (then, Output_Get:float)
   > "Set GridSize (Read-Only)"("GridSize (Read-Only)":int=Truncate(A=@n14)) -> (Output_Get:int=0)
   n15 = GetScaledBoxExtent(self=Get Box())
   n16 = Get Actor Location!(self) -> (ReturnValue:Vector)
   Set GridOrigin(execute=@n4.then_4, GridOrigin:Vector 2D Structure="To Vector2D (Vector)"(InVector=vector - vector(A=@n16, B=@n15))) -> (Output_Get:Vector 2D Structure)
  ]}
 WarmUp {nodes [
   WarmUp() {is=title K2Node_FunctionEntry}
   > Branch(Condition=Get WarmUpCapture() -> (WarmUpCapture:bool))
   n1 = Get Object Path String(Object=Self-Reference())
   > Branch(Condition=NOT Boolean(A=StartsWith(SourceString=@n1, InPrefix=/Engine)))
   > CaptureScene(self=Get SceneCaptureComponent2D())
  ]}
]
variables [
 AssetNameSuffixes {type="array<string>", default []}
 BoxTop {type=number}
 BucketCoordinates {type="array<Vector>"}
 BucketSize {type=number, default=1}
 CaptureHeightBias {type=number}
 CellExtent {type=number}
 CellSize {type=number}
 CurrentTexture {type=Texture2D}
 Debug {type=boolean}
 DebugRange {type=IntPoint, default=<0 0>}
 DynamicMeshes {type="array<DynamicMesh>", default []}
 ExcludedMaterials {type="array<MaterialInterface>", default []}
 ExcludedMaterials_Internal {type="array<MaterialInterface>", default []}
 GeneratedMeshes {type="array<StaticMesh>", default []}
 GeneratedTextures {type="array<Texture2D>", default []}
 GridOrigin {type=Vector2D, default=<0 0>}
 GridSize {type=number}
 "GridSize (Read-Only)" {type=integer}
 IncludedMaterials {type="array<MaterialInterface>", default []}
 IncludedMaterials_Internal {type="array<MaterialInterface>", default []}
 LoopIndex {type=integer}
 MeshBiasZ {type=number}
 MeshData {type="array<DistantFoliage_MeshData>", default []}
 MeshDilation {type=number}
 MeshResolution {type=integer, default=16}
 MultiSample {type=boolean}
 MultiSampleCount {type=integer, default=4}
 MultiSampleJitter {type=number, default=100}
 NumActiveBuckets {type=integer}
 NumBuckets {type=integer, default=5}
 NumBucketsFloat {type=number, default=5}
 OutputActors {type=boolean, default=true}
 OutputMeshAssets {type=boolean, default=true}
 OutputMeshes {type="array<StaticMesh>", default []}
 OutputPath {type=string, default=""}
 OutputTextureAssets {type=boolean, default=true}
 OutputTextures {type="array<Texture2D>", default []}
 PostProcessTextures {type=boolean, default=true}
 RenderActorData {type="array<DistantFoliage_RenderActorData>", default []}
 RenderedActors {type="array<BP_MeshGrass_C>", default []}
 RenderResolution {type=integer, default=1024}
 ReprojectMissedTraces {type=boolean, default=true}
 ResampleCliffs {type=boolean}
 ResampleRadius {type=number, default=0.5}
 RT_Capture {type=TextureRenderTarget2D, default=null}
 RT_Output {type=TextureRenderTarget2D, default=null}
 SelectiveBuckets {type="array<integer>"}
 ShowWireframe {type=boolean, default=true}
 TextureResolution {type=integer, default=128}
 TotalBuckets {type=number, default=5}
 UseDebugRange {type=boolean}
 UseExcludeList {type=boolean}
 UseSelectiveBuckets {type=boolean}
 Verbose {type=boolean}
 VertCounts {type="array<integer>"}
 WarmUpCapture {type=boolean}
]
