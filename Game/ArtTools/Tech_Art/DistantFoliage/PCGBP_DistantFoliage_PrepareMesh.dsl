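; PCG Blueprint node that prepares point data as mesh data for the distant foliage generator (EUBP_DistantFoliage_Generator) that owns the PCG component.
; It copies bucket coordinates from the BucketAttributes param metadata to the generator.
; A point loop over each point input records 3D and 2D vertex positions plus an index filter based on zero density, stripping -1 entries.
; The resulting per-input structs go to the generator, and the looped points are output.
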(fn ExecuteWithContext
  (params (InContext :type PCGContext%20Structure%20%28by%20ref%29) (Input :type PCGData%20Collection%20Structure%20%28by%20ref%29))
  (Utilities|FlowControl|Sequence
    (:then_0
      (bind _returnvalue (PCG|Temporary|GetComponent (Variables|GetInContext)))
      (Variables|Default|SetOwner (Utilities|Casting|CastToEUBP_DistantFoliage_Generator (Components|GetOwner _returnvalue))))
    (:then_1
      (bind (_outtaggeddata _returnvalue_2) (Data|GetTypedInputsbyPinLabel (Variables|GetInput) "BucketAttributes" "/Script/PCG.PCGParamData"))
      (if (Utilities|Array|IsNotEmpty _returnvalue_2)
        (if (> (Math|Conversions|ToInteger(Integer64) (PCG|Metadata|GetNumberofEntries (Metadata|ConstMetadata (Utilities|Array|Get(acopy) _returnvalue_2)))) 0)
          (Utilities|Array|Resize (Variables|Default|GetBucketCoordinates) (Math|Conversions|ToInteger(Integer64) (PCG|Metadata|GetNumberofEntries (Metadata|ConstMetadata (Utilities|Array|Get(acopy) _returnvalue_2)))))
          (for _index (range (Math|Conversions|ToInteger(Integer64) (PCG|Metadata|GetNumberofEntries (Metadata|ConstMetadata (Utilities|Array|Get(acopy) _returnvalue_2)))))
            (bind _returnvalue_3 (PCG|Metadata|GetVectorAttributebyMetadataKey (Math|Conversions|ToInteger64(Integer) _index) (Metadata|ConstMetadata (Utilities|Array|Get(acopy) _returnvalue_2)) "BucketCoordinates"))
            (Utilities|Array|SetArrayElem (Variables|Default|GetBucketCoordinates) _index _returnvalue_3))
          (Variables|/Game/ArtTools/Tech_Art/DistantFoliage/EUBP_DistantFoliage_Generator.EUBP_DistantFoliage_Generator_C|SetBucketCoordinates (Variables|Default|GetBucketCoordinates) (Variables|Default|GetOwner))
          (else
            _)))
      (Development|PrintString "Error: Missing BucketAttributes"))
    (:then_2
      (bind (_outtaggeddata_2 _returnvalue_4) (Data|GetTypedInputs (Variables|GetInput) "/Script/PCG.PCGPointData"))
      (Utilities|Array|Clear (Variables|Default|GetMeshData))
      (for _array_element _returnvalue_4
        (Utilities|Array|Resize (Variables|Default|GetVertPos3D) (PointData|GetNumPoints _array_element))
        (Utilities|Array|Resize (Variables|Default|GetVertPos2D) (PointData|GetNumPoints _array_element))
        (Utilities|Array|Resize (Variables|Default|GetVertFilter) (PointData|GetNumPoints _array_element))
        (bind _outdata (PCG|FlowControl|PointLoop (Variables|GetInContext) _array_element 0))
        (Utilities|Array|Add (LocalVariables|GetOutData) (PCG|TaggedData|MakeTaggedData _outdata))
        (Utilities|Array|RemoveItem (Variables|Default|GetVertFilter) -1)
        (Utilities|Array|Add (Variables|Default|GetMeshData) (Utilities|Struct|MakeDistantFoliageMeshData (Variables|Default|GetVertPos3D) (Variables|Default|GetVertPos2D) (Variables|Default|GetVertFilter))))
      (Variables|/Game/ArtTools/Tech_Art/DistantFoliage/EUBP_DistantFoliage_Generator.EUBP_DistantFoliage_Generator_C|SetMeshData (Variables|Default|GetMeshData) (Variables|Default|GetOwner))
      (return (Utilities|Struct|MakePCGDataCollection (LocalVariables|GetOutData))))))

(fn PointLoopBody
  (params (InContext :type PCGContext%20Structure%20%28by%20ref%29) (InData :type PCGPoint%20Data%20Object%20Reference) (InPoint :type PCGPoint%20Structure%20%28by%20ref%29) (OutMetadata :type PCGMetadata%20Object%20Reference) (Iteration :type Integer64))
  (bind (_transform _density) (Utilities|Struct|BreakPCGPoint InPoint))
  (bind _transform_location (Math|Transform|BreakTransform _transform))
  (bind _returnvalue (Math|Conversions|ToInteger(Integer64) Iteration))
  (Utilities|Array|SetArrayElem (Variables|Default|GetVertPos3D) _returnvalue _transform_location)
  (Utilities|Array|SetArrayElem (Variables|Default|GetVertPos2D) _returnvalue (Math|Conversions|ToVector2D(Vector) _transform_location))
  (Utilities|Array|SetArrayElem (Variables|Default|GetVertFilter) _returnvalue (select (== _density 0) _returnvalue -1))
  (return (Variables|GetInPoint) true))
