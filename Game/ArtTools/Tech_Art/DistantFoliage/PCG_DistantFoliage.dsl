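; PCG graph that builds per-bucket point data for distant foliage meshes, passing it to a PrepareMesh Blueprint.
; It grids the actor's bounds into buckets, optionally debug-filtered, runs BucketProcessor on valid generated textures, then raycasts samples onto terrain.
; Optional cliff resampling ring-traces samples near cliff bounds, using landscape-hit and normal tests to adjust positions.
; Optional reprojection recovers missed-trace heights; coverage becomes density, and positions become bucket-relative.
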(graph PCG_DistantFoliage
  (params (CliffProcessing :type Bool :default true))
  (outputs (OutPoints :type Point :advanced))
  (bind AttributePartition_4 (Attribute%20Partition :PartitionAttributeSelectors [BucketIndex]))
  (bind GetAttributeFromPointIndex_6 (Get%20Attribute%20From%20Point%20Index :bForceOutputAttributeToBeInElementsDomain true))
  (bind GetAttributeFromPointIndex_8 (Get%20Attribute%20From%20Point%20Index :bForceOutputAttributeToBeInElementsDomain true))
  ;[0.15 0.15 0.15 0.5] MeshData Generation
  (bind CreateAttribute_15 (Create%20Constant :AttributeTypes (struct :DoubleValue 0.5)))
  (bind DeleteAttributes_7 (Delete%20Attributes :SelectedAttributes "BoundsMin"))
  (bind AttributeMathsOp_13 (Attribute%20Maths%20Op :InputSource1 BoundsMax.Z :Operation Set))
  (bind CreateAttribute_11 (Create%20Constant :AttributeTypes (struct :DoubleValue 1.0)))
  (bind MakeVectorAttribute_10 (Make%20Vector%20Attribute :OutputType Vector))
  (bind CreatePointsGrid_3 (Create%20Points%20Grid :CellSize (1e+02 1e+02 1e+06) :CoordinateSpace OriginalComponent :GridExtents (5e+02 5e+02 1.0) :PointSteepness 0.5))
  (bind Branch_14 (Branch))
  (bind Branch_22 (Branch))
  (bind Branch_27 (Branch))
  (bind CreateAttribute_18 (Create%20Constant :AttributeTypes (struct :Type Vector :VectorValue (0.0 0.0 1e+04))))
  (bind CreateAttribute_12 (Create%20Constant :AttributeTypes (struct :Type Integer32 :Int32Value 1 :DoubleValue 1.0)))
  (bind AttributeMathsOp_11 (Attribute%20Maths%20Op :InputSource1 Y))
  (bind AttributeCast_7 (Attribute%20Cast :InputSource Y :OutputType String))
  (bind AttributeCast_1 (Attribute%20Cast :InputSource X :OutputType String))
  (bind CreateAttribute_3 (Create%20Constant :AttributeTypes (struct :Type String :StringValue ":")))
  (bind AttributeStringOp_2 (Attribute%20String%20Op))
  (bind AttributeStringOp_5 (Attribute%20String%20Op))
  (bind DeleteAttributes_8 (Delete%20Attributes :SelectedAttributes "Y"))
  (bind FilterElementsByIndex_9 (Filter%20Elements%20By%20Index :bSelectIndicesByInput false :IndexSelectionAttribute X))
  (bind AttributeCast_31 (Attribute%20Cast :OutputType String))
  (bind ExecuteBlueprint_30 (Execute%20Blueprint :BlueprintElementInstance (class:/Game/Athena/Helios/Environments/Landscape/WorldBuildingTools/PCGGrass/PCGBP_JoinStrings.PCGBP_JoinStrings_C :TargetAttribute SelectiveBuckets) :BlueprintElementType /Game/Athena/Helios/Environments/Landscape/WorldBuildingTools/PCGGrass/PCGBP_JoinStrings.PCGBP_JoinStrings_C))
  (bind FilterElementsByIndex_25 (Filter%20Elements%20By%20Index :bSelectIndicesByInput false :IndexSelectionAttribute X))
  (bind AttributeMathsOp_17 (Attribute%20Maths%20Op :InputSource1 $Position))
  (bind AttributeNoise_12 (Attribute%20Noise))
  (bind AttributePartition_19 (Attribute%20Partition))
  (bind Subgraph_0 (Subgraph :Graph /Game/Athena/Helios/Environments/Landscape/WorldBuildingTools/PCGGrass/PCG_GateValidEntries.PCG_GateValidEntries))
  (bind AttributePartition_43 (Attribute%20Partition))
  (bind CopyAttributes_42 (Copy%20Attributes :bCopyAllAttributes true :InputSource $Density :OutputTarget None))
  (bind Loop_24 (Loop :Graph /Game/ArtTools/Tech_Art/DistantFoliage/PCG_DistantFoliage_BucketProcessor.PCG_DistantFoliage_BucketProcessor))
  (bind Branch_98 (Branch))
  (bind WorldRaycast_26 (World%20Raycast :bKeepOriginalPointOnMiss true :WorldQueryParams (struct :ActorClassFilter Include :ActorClass /Script/Landscape.LandscapeStreamingProxy :ActorFilterFromInput Include :ActorFilterInputSource ActorReference)))
  (bind AttributeMathsOp_0 (Attribute%20Maths%20Op :InputSource1 $Position.XY :Operation Round))
  (bind Branch_8 (Branch))
  (bind AttributeFilter_27 (Filter%20Attribute%20Elements :bUseConstantThreshold true :Operator Equal :ThresholdAttribute @Last))
  (bind AddAttribute_31 (Add%20Attribute :AttributeTypes (struct :DoubleValue -1.0) :OutputTarget PosZ))
  (bind CreateAttribute_34 (Create%20Constant :AttributeTypes (struct :DoubleValue 1.0)))
  (bind AttributeMathsOp_33 (Attribute%20Maths%20Op :InputSource1 $Density :Operation Set))
  (bind CreateAttribute_19 (Create%20Constant :AttributeTypes (struct :DoubleValue 2.0)))
  (bind GetAttributeFromPointIndex_11 (Get%20Attribute%20From%20Point%20Index :bForceOutputAttributeToBeInElementsDomain true :InputSource $Position.Z :OutputAttributeName None))
  (bind AttributeMathsOp_18 (Attribute%20Maths%20Op :Operation Multiply))
  (bind CreateAttribute_21 (Create%20Constant :AttributeTypes (struct :DoubleValue 2e+03)))
  (bind MakeVectorAttribute_20 (Make%20Vector%20Attribute :OutputType Vector))
  (bind TransformPoints_25 (Transform%20Points :bAbsoluteRotation true))
  (bind ExtentsModifier_16 (Extents%20Modifier))
  (bind DensityFilter_24 (Density%20Filter :LowerBound 1.0))
  (bind MergePoints_29 (Merge%20Points))
  (bind CopyAttributes_30 (Copy%20Attributes :InputSource $Position.Z :OutputTarget PosZ))
  (bind Projection_26 (Projection :ProjectionParams (struct :bProjectPositions false :bProjectRotations false :AttributeList "PosZ" :AttributeMode IncludeAttributes)))
  (bind AttributeMathsOp_35 (Attribute%20Maths%20Op :InputSource1 $Position.Z :InputSource2 PosZ :Operation Set)) ; Miss Reprojection
  (bind MergePoints_47 (Merge%20Points))
  (bind SortAttributes_1 (Sort%20Attributes :bUseStableSort false :SortAttributes [(struct :InputSource BucketIndex)]))
  (bind AttributePartition_39 (Attribute%20Partition :PartitionAttributeSelectors [BucketIndex]))
  (bind Gather_10 (Gather))
  (bind AttributeMathsOp_44 (Attribute%20Maths%20Op :InputSource1 $Density :InputSource2 Coverage :Operation Set))
  (bind AttributePartition_18 (Attribute%20Partition :PartitionAttributeSelectors [$Index]))
  (bind AttributeMathsOp_19 (Attribute%20Maths%20Op :InputSource1 $Position.XY :InputSource2 BucketCoordinates.XY :Operation Subtract))
  (bind ExecuteBlueprint_4 (Execute%20Blueprint :BlueprintElementInstance (class:/Game/ArtTools/Tech_Art/DistantFoliage/PCGBP_DistantFoliage_PrepareMesh.PCGBP_DistantFoliage_PrepareMesh_C) :BlueprintElementType /Game/ArtTools/Tech_Art/DistantFoliage/PCGBP_DistantFoliage_PrepareMesh.PCGBP_DistantFoliage_PrepareMesh_C))
  (bind MergePoints_38 (Merge%20Points))
  (bind AttributeFilter_1 (Filter%20Attribute%20Elements :bUseConstantThreshold true :Operator NotEqual :ThresholdAttribute @Last))
  (bind Gather_4 (Gather))
  (bind CopyAttributes_6 (Copy%20Attributes :InputSource $Position :OutputTarget BucketCoordinates))
  (bind PointToAttributeSet_15 (Point%20To%20Attribute%20Set))
  (bind AttributeMathsOp_20 (Attribute%20Maths%20Op :InputSource1 @Last.Z :Operation Set))
  (bind AddAttribute_0 (Add%20Attribute :bCopyAllAttributes true))
  (bind CreateAttribute_49 (Create%20Constant :AttributeTypes (struct :DoubleValue 0.1)))
  (bind AttributeMathsOp_48 (Attribute%20Maths%20Op :InputSource1 $Density :Operation Set))
  (bind ExtentsModifier_45 (Extents%20Modifier :Extents (0.5 0.5 0.25) :Mode Multiply))
  (bind GetAttributeFromPointIndex_121 (Get%20Attribute%20From%20Point%20Index :bForceOutputAttributeToBeInElementsDomain true))
  (bind AttributeNoise_120 (Attribute%20Noise :InputSource $Color :OutputTarget $Color))
  (bind CopyAttributes_122 (Copy%20Attributes :InputSource $Color :OutputTarget $Color))
  ;[0.15 0.15 0.15 0.5] Inputs
  (bind GetActorData_9 (Get%20Actor%20Data :Mode GetActorReference))
  (bind GetPropertyFromObjectPath_8 (Get%20Property%20From%20Object%20Path :bSanitizeOutputAttributeName false :OutputAttributeName @SourceName :PropertyName CellSize))
  (bind GetPropertyFromObjectPath_37 (Get%20Property%20From%20Object%20Path :bSanitizeOutputAttributeName false :OutputAttributeName @SourceName :PropertyName GeneratedTextures))
  (bind GetActorData_39 (Get%20Actor%20Data :bIgnorePCGGeneratedComponents false :bTrackActorsOnlyWithinBounds false :ComponentSelector (struct :ComponentSelection ByClass :ComponentSelectionClass /Script/Engine.BoxComponent)))
  (bind GetBounds_41 (Get%20Bounds))
  (bind GetPropertyFromObjectPath_14 (Get%20Property%20From%20Object%20Path :bSanitizeOutputAttributeName false :OutputAttributeName @SourceName :PropertyName MeshResolution))
  (bind GetPropertyFromObjectPath_3 (Get%20Property%20From%20Object%20Path :bSanitizeOutputAttributeName false :PropertyName DebugRange))
  (bind GetPropertyFromObjectPath_5 (Get%20Property%20From%20Object%20Path :bSanitizeOutputAttributeName false :OutputAttributeName Debug :PropertyName Debug))
  (bind GetActorData_0 (Get%20Actor%20Data :ActorSelector (struct :ActorFilter AllWorldActors :ActorSelection ByClass :ActorSelectionClass /Hermes/Terrain/Cliffs/Blueprints/BP_Hermes_Cliff_PARENT.BP_Hermes_Cliff_PARENT_C :bSelectMultiple true) :bMergeSinglePointData true :bTrackActorsOnlyWithinBounds false :Mode GetSinglePoint))
  (bind GetPropertyFromObjectPath_22 (Get%20Property%20From%20Object%20Path :bSanitizeOutputAttributeName false :PropertyName MeshBiasZ))
  (bind GetPropertyFromObjectPath_125 (Get%20Property%20From%20Object%20Path :bSanitizeOutputAttributeName false :OutputAttributeName @SourceName :PropertyName MeshDilation))
  (bind GetPropertyFromObjectPath_128 (Get%20Property%20From%20Object%20Path :bSanitizeOutputAttributeName false :OutputAttributeName @SourceName :PropertyName ResampleCliffs))
  (bind GetPropertyFromObjectPath_141 (Get%20Property%20From%20Object%20Path :bSanitizeOutputAttributeName false :OutputAttributeName @SourceName :PropertyName ResampleRadius))
  (bind GetPropertyFromObjectPath_10 (Get%20Property%20From%20Object%20Path :bSanitizeOutputAttributeName false :PropertyName SelectiveBuckets))
  (bind GetPropertyFromObjectPath_15 (Get%20Property%20From%20Object%20Path :bSanitizeOutputAttributeName false :PropertyName UseSelectiveBuckets))
  (bind GetPropertyFromObjectPath_18 (Get%20Property%20From%20Object%20Path :bSanitizeOutputAttributeName false :PropertyName UseDebugRange))
  (bind GetPropertyFromObjectPath_12 (Get%20Property%20From%20Object%20Path :bSanitizeOutputAttributeName false :OutputAttributeName @SourceName :PropertyName ReprojectMissedTraces))
  ;[0.05 0.05 0.05] Cliff Processing
  (bind AttributePartition_56 (Attribute%20Partition :PartitionAttributeSelectors [$Index]))
  (bind GetBounds_54 (Get%20Bounds))
  (bind MergeAttributes_64 (Merge%20Attributes))
  (bind TransformPoints_63 (Transform%20Points :bAbsoluteRotation true :bAbsoluteScale true))
  (bind AttributeMathsOp_60 (Attribute%20Maths%20Op :InputSource1 $BoundsMax :InputSource2 BoundsMax :Operation Set))
  (bind AttributeMathsOp_62 (Attribute%20Maths%20Op :InputSource1 $BoundsMin :InputSource2 BoundsMin :Operation Set))
  (bind AttributeMathsOp_57 (Attribute%20Maths%20Op :InputSource1 $BoundsMax :InputSource2 $Position :Operation Subtract))
  (bind AttributeMathsOp_59 (Attribute%20Maths%20Op :InputSource1 $BoundsMin :InputSource2 $Position :Operation Subtract))
  (bind CreateAttribute_67 (Create%20Constant :AttributeTypes (struct :DoubleValue 3e+04)))
  (bind AttributeMathsOp_65 (Attribute%20Maths%20Op :InputSource1 $BoundsMax.Z :Operation Set)) ; Cliff AABBs
  (bind Branch_103 (Branch))
  (bind Difference_68 (Difference :bDiffMetadata false :bKeepZeroDensityPoints true :DensityFunction Binary :Mode Discrete))
  (bind ToPoint_70 (To%20Point))
  (bind AttributeFilter_69 (Filter%20Attribute%20Elements :bUseConstantThreshold true :Operator Equal))
  (bind CreateAttribute_72 (Create%20Constant :AttributeTypes (struct :DoubleValue 1.0)))
  (bind MergePoints_0 (Merge%20Points))
  (bind AttributeMathsOp_71 (Attribute%20Maths%20Op :InputSource1 $Density :Operation Set))
  (bind CreatePoints_73 (Create%20Points :PointsToCreate [(struct :Transform (struct :Translation (1e+02 0.0 0.0))) (struct :Transform (struct :Translation (-5e+01 86.6 0.0))) (struct :Transform (struct :Translation (-5e+01 -86.6 0.0)))]))
  (bind CreatePoints_74 (Create%20Points :PointsToCreate [(struct :Transform (struct :Translation (1e+02 0.0 0.0))) (struct :Transform (struct :Translation (30.9 95.1 0.0))) (struct :Transform (struct :Translation (-80.9 58.78 0.0))) (struct :Transform (struct :Translation (-80.9 -58.78 0.0))) (struct :Transform (struct :Translation (30.9 -95.1 0.0)))]))
  (bind CreateAttribute_88 (Create%20Constant :AttributeTypes (struct :DoubleValue 1e+02)))
  (bind FilterElementsByIndex_93 (Filter%20Elements%20By%20Index :bSelectIndicesByInput false :IndexSelectionAttribute $Index :SelectedIndices "1"))
  (bind DeleteAttributes_89 (Delete%20Attributes :Operation KeepSelectedAttributes :SelectedAttributes "CellSize"))
  (bind AttributeMathsOp_86 (Attribute%20Maths%20Op :InputSource1 CellSize :Operation Divide))
  (bind CreateAttribute_97 (Create%20Constant :AttributeTypes (struct :DoubleValue 2.0)))
  (bind AttributeMathsOp_95 (Attribute%20Maths%20Op :InputSource1 CellSize :Operation Divide))
  (bind AttributeMathsOp_139 (Attribute%20Maths%20Op :Operation Multiply))
  (bind CreatePoints_80 (Create%20Points :PointsToCreate [(struct :Transform (struct :Translation (1e+02 0.0 0.0))) (struct :Transform (struct :Translation (70.71 70.71 0.0))) (struct :Transform (struct :Translation (0.0 1e+02 0.0))) (struct :Transform (struct :Translation (-70.71 70.71 0.0))) (struct :Transform (struct :Translation (-1e+02 0.0 0.0))) (struct :Transform (struct :Translation (-70.71 -70.71 0.0))) (struct :Transform (struct :Translation (0.0 -1e+02 0.0))) (struct :Transform (struct :Translation (70.71 -70.71 0.0))) (struct :Steepness 1.0)]))
  (bind AttributeMathsOp_76 (Attribute%20Maths%20Op :InputSource1 $Position :Operation Multiply))
  (bind CreateAttribute_84 (Create%20Constant :AttributeTypes (struct :DoubleValue 5e+01)))
  (bind ExtentsModifier_77 (Extents%20Modifier))
  (bind CopyAttributes_2 (Copy%20Attributes :InputSource $Index :OutputTarget ID))
  (bind CopyPoints_78 (Copy%20Points :AttributeInheritance TargetOnly :ColorInheritance Source :RotationInheritance Target :ScaleInheritance Source :SeedInheritance Target :TagInheritance Source))
  (bind CreatePoints_79 (Create%20Points :PointsToCreate [(struct :Transform (struct :Translation (1e+02 0.0 0.0))) (struct :Transform (struct :Translation (5e+01 86.6 0.0))) (struct :Transform (struct :Translation (-5e+01 86.6 0.0))) (struct :Transform (struct :Translation (-1e+02 0.0 0.0))) (struct :Transform (struct :Translation (-5e+01 -86.6 0.0))) (struct :Transform (struct :Translation (5e+01 -86.6 0.0)))]))
  (bind WorldRaycast_99 (World%20Raycast :bKeepOriginalPointOnMiss true :WorldQueryParams (struct :ActorClassFilter Include :ActorClass /Script/Landscape.LandscapeStreamingProxy :ActorFilterFromInput Include :ActorFilterInputSource ActorReference :bGetReferenceToActorHit true))) ; Collect Samples in Radius
  (bind AttributePartition_8 (Attribute%20Partition :PartitionAttributeSelectors [ActorReference]))
  (bind GetAttributeFromPointIndex_9 (Get%20Attribute%20From%20Point%20Index :bForceOutputAttributeToBeInElementsDomain true :OutputAttributeName None))
  (bind ExecuteBlueprint_6 (Execute%20Blueprint :BlueprintElementInstance (class:/Game/Athena/Helios/Environments/Landscape/WorldBuildingTools/PCGGrass/PCGBP_StringContains.PCGBP_StringContains_C :Substring "Landscape") :BlueprintElementType /Game/Athena/Helios/Environments/Landscape/WorldBuildingTools/PCGGrass/PCGBP_StringContains.PCGBP_StringContains_C))
  (bind CopyAttributes_10 (Copy%20Attributes :InputSource @Last :OutputTarget IsLandscape))
  (bind MergePoints_11 (Merge%20Points))
  (bind AttributePartition_21 (Attribute%20Partition :PartitionAttributeSelectors [ID]))
  (bind AttributeReduce_22 (Attribute%20Reduce :InputSource IsLandscape :Operation Min :OutputAttributeName AllLandscape))
  (bind CopyAttributes_24 (Copy%20Attributes :bCopyAllAttributes true :InputSource AllLandscape :OutputTarget AllLandscape))
  (bind AttributeFilter_26 (Filter%20Attribute%20Elements :AttributeTypes (struct :Type Boolean :BoolValue true) :bUseConstantThreshold true :Operator Equal :TargetAttribute AllLandscape :ThresholdAttribute @Last))
  (bind MergePoints_57 (Merge%20Points))
  (bind AttributePartition_59 (Attribute%20Partition :PartitionAttributeSelectors [ID]))
  (bind AttributeFilter_12 (Filter%20Attribute%20Elements :AttributeTypes (struct :Type Boolean :BoolValue true) :bUseConstantThreshold true :Operator NotEqual :TargetAttribute IsLandscape :ThresholdAttribute @Last))
  (bind AttributePartition_3 (Attribute%20Partition :PartitionAttributeSelectors [ID]))
  (bind NormalToDensity_4 (Normal%20To%20Density :Strength 0.5))
  (bind CreateAttribute_62 (Create%20Constant :AttributeTypes (struct :DoubleValue 0.8)))
  (bind AttributeCompareOp_61 (Attribute%20Compare%20Op :InputSource1 $Density :Operation Less :OutputTarget NormalFail))
  (bind AttributeReduce_64 (Attribute%20Reduce :InputSource NormalFail :Operation Min :OutputAttributeName AllNormalFail))
  (bind CopyAttributes_66 (Copy%20Attributes :bCopyAllAttributes true :InputSource $Density :OutputTarget None))
  (bind AttributeFilter_68 (Filter%20Attribute%20Elements :AttributeTypes (struct :Type Boolean :BoolValue true) :bUseConstantThreshold true :Operator Equal :TargetAttribute AllNormalFail :ThresholdAttribute @Last))
  (bind MergePoints_70 (Merge%20Points))
  (bind AttributeFilter_135 (Filter%20Attribute%20Elements :AttributeTypes (struct :Type Boolean) :bUseConstantThreshold true :Operator Equal :TargetAttribute NormalFail :ThresholdAttribute @Last))
  (bind AttributePartition_72 (Attribute%20Partition :PartitionAttributeSelectors [ID]))
  (bind AttributeReduce_14 (Attribute%20Reduce :InputSource $Position :OutputAttributeName Pos))
  (bind AttributeReduce_137 (Attribute%20Reduce :InputSource $Position.Z :Operation Max :OutputAttributeName Pos))
  (bind AttributeMathsOp_138 (Attribute%20Maths%20Op :InputSource1 Pos.Z :Operation Set))
  (bind GetAttributeFromPointIndex_16 (Get%20Attribute%20From%20Point%20Index :bForceOutputAttributeToBeInElementsDomain true))
  (bind CopyAttributes_15 (Copy%20Attributes :InputSource Pos :OutputTarget $Position)) ; Processed
  (bind AttributeFilter_149 (Filter%20Attribute%20Elements :AttributeTypes (struct :Type Boolean :BoolValue true) :bUseConstantThreshold true :Operator Equal :TargetAttribute NormalFail :ThresholdAttribute @Last))
  (bind AttributePartition_167 (Attribute%20Partition :PartitionAttributeSelectors [ID]))
  (bind AttributeReduce_151 (Attribute%20Reduce :InputSource $Position :OutputAttributeName PosXY))
  (bind CopyAttributes_163 (Copy%20Attributes :InputSource ID))
  (bind MergeAttributes_178 (Merge%20Attributes))
  (bind MatchAndSetAttributes_171 (Match%20And%20Set%20Attributes :bKeepUnmatched false :bMatchAttributes true :InputAttribute ID :MatchAttribute ID))
  (bind MergePoints_158 (Merge%20Points))
  (bind AttributePartition_147 (Attribute%20Partition :PartitionAttributeSelectors [ID]))
  (bind AttributeReduce_156 (Attribute%20Reduce :InputSource $Position.Z :Operation Max :OutputAttributeName PosZ))
  (bind CopyAttributes_165 (Copy%20Attributes :InputSource ID))
  (bind MergeAttributes_180 (Merge%20Attributes))
  (bind MatchAndSetAttributes_173 (Match%20And%20Set%20Attributes :bKeepUnmatched false :bMatchAttributes true :InputAttribute ID :MatchAttribute ID))
  (bind CopyAttributes_175 (Copy%20Attributes :InputSource PosXY :OutputTarget $Position))
  (bind CopyAttributes_177 (Copy%20Attributes :InputSource PosZ :OutputTarget $Position.Z)) ; Processed
  (bind ExtentsModifier_17 (Extents%20Modifier :Extents (1.0 1.0 4.0) :Mode Multiply))
  (bind MergePoints_36 (Merge%20Points))
  (bind AttributePartition_38 (Attribute%20Partition :PartitionAttributeSelectors [ID]))
  (bind GetAttributeFromPointIndex_43 (Get%20Attribute%20From%20Point%20Index :bForceOutputAttributeToBeInElementsDomain true :Index 8)) ; Landscape Rejection
  (bind ExtentsModifier_35 (Extents%20Modifier :Extents (0.5 0.5 8.0) :Mode Multiply))
  (bind GetAttributeFromPointIndex_79 (Get%20Attribute%20From%20Point%20Index :bForceOutputAttributeToBeInElementsDomain true :Index 8))
  (bind MergePoints_88 (Merge%20Points))
  (bind CopyAttributes_85 (Copy%20Attributes :InputSource $Position :OutputTarget OriginalPos)) ; Sample Group Centers
  (bind PointToAttributeSet_86 (Point%20To%20Attribute%20Set))
  (bind DeleteAttributes_87 (Delete%20Attributes :Operation KeepSelectedAttributes :SelectedAttributes "ID,OriginalPos"))
  (bind MergePoints_74 (Merge%20Points))
  (bind AttributePartition_76 (Attribute%20Partition :PartitionAttributeSelectors [ID]))
  (bind GetAttributeFromPointIndex_81 (Get%20Attribute%20From%20Point%20Index :bForceOutputAttributeToBeInElementsDomain true))
  (bind MergePoints_82 (Merge%20Points))
  (bind MatchAndSetAttributes_84 (Match%20And%20Set%20Attributes :bKeepUnmatched false :bMatchAttributes true :InputAttribute ID :MatchAttribute ID))
  (bind AttributeMathsOp_89 (Attribute%20Maths%20Op :InputSource1 $Position :InputSource2 OriginalPos :Operation Set)) ; Normal Rejection
  (bind MergePoints_39 (Merge%20Points))
  (bind CreateAttribute_53 (Create%20Constant :AttributeTypes (struct :Type Vector :VectorValue (1.0 0.0 0.0))))
  (bind AttributeMathsOp_51 (Attribute%20Maths%20Op :InputSource1 $Color :Operation Set))
  (bind AttributePartition_145 (Attribute%20Partition :PartitionAttributeSelectors [ID]))
  (bind MergePoints_5 (Merge%20Points))
  (reroute Bounds GetBounds_41)
  (reroute Bucket%20Elements Loop_24.Buckets)
  (reroute BucketAttributes AddAttribute_0)
  (reroute Buckets AttributeMathsOp_17)
  (reroute CellSize GetPropertyFromObjectPath_8)
  (reroute Cliff%20Processed MergePoints_39)
  (reroute CliffProcessingCandidates AttributeFilter_69.InsideFilter)
  (reroute CliffProcessingSkip AttributeFilter_69.OutsideFilter)
  (reroute Cliffs GetActorData_0)
  (reroute CliffSamplesRaw WorldRaycast_99)
  (reroute Debug GetPropertyFromObjectPath_5)
  (reroute DebugRange GetPropertyFromObjectPath_3)
  (reroute GeneratedTextures GetPropertyFromObjectPath_37)
  (reroute MeshBiasZ GetPropertyFromObjectPath_22)
  (reroute MeshDilation GetPropertyFromObjectPath_125)
  (reroute MeshResolution GetPropertyFromObjectPath_14)
  (reroute ReprojectMissedTraces GetPropertyFromObjectPath_12)
  (reroute ResampleCliffs GetPropertyFromObjectPath_128)
  (reroute ResampleRadius GetPropertyFromObjectPath_141)
  (reroute SelectiveBuckets GetPropertyFromObjectPath_10)
  (reroute UseDebugRange GetPropertyFromObjectPath_18)
  (reroute UseSelectiveBuckets GetPropertyFromObjectPath_15)
  (connect Loop_24 BucketAttributes AddAttribute_0 Attributes)
  (connect AttributeMathsOp_20 Out AddAttribute_0 In)
  (connect AttributeFilter_27 InsideFilter AddAttribute_31 In)
  (connect DebugRange Out AttributeCast_1 In)
  (connect AttributeMathsOp_11 Out AttributeCast_7 In)
  (connect SelectiveBuckets Out AttributeCast_31 In)
  (connect NormalToDensity_4 Out AttributeCompareOp_61 InA)
  (connect CreateAttribute_62 Out AttributeCompareOp_61 InB)
  (connect AttributeMathsOp_44 Out AttributeFilter_1 In)
  (connect AttributePartition_59 Out AttributeFilter_12 In)
  (connect CopyAttributes_24 Out AttributeFilter_26 In)
  (connect Branch_8 Output%20B AttributeFilter_27 In)
  (connect CopyAttributes_66 Out AttributeFilter_68 In)
  (connect ToPoint_70 Out AttributeFilter_69 In)
  (connect MergePoints_70 Out AttributeFilter_135 In)
  (connect MergePoints_70 Out AttributeFilter_149 In)
  (connect WorldRaycast_26 Out AttributeMathsOp_0 In)
  (connect DebugRange Out AttributeMathsOp_11 InA)
  (connect CreateAttribute_12 Out AttributeMathsOp_11 InB)
  (connect DeleteAttributes_7 Out AttributeMathsOp_13 InA)
  (connect CreateAttribute_15 Out AttributeMathsOp_13 InB)
  (connect Branch_14 Output%20A AttributeMathsOp_17 InA)
  (connect Branch_27 Output%20A AttributeMathsOp_17 InA)
  (connect FilterElementsByIndex_9 In%20Filter AttributeMathsOp_17 InA)
  (connect FilterElementsByIndex_25 In%20Filter AttributeMathsOp_17 InA)
  (connect CreateAttribute_18 Out AttributeMathsOp_17 InB)
  (connect GetAttributeFromPointIndex_11 Attribute AttributeMathsOp_18 InA)
  (connect CreateAttribute_19 Out AttributeMathsOp_18 InB)
  (connect AttributeMathsOp_44 Out AttributeMathsOp_19 InA)
  (connect AttributePartition_18 Out AttributeMathsOp_19 InB)
  (connect PointToAttributeSet_15 Out AttributeMathsOp_20 InA)
  (connect MeshBiasZ Out AttributeMathsOp_20 InB)
  (connect AddAttribute_31 Out AttributeMathsOp_33 InA)
  (connect CreateAttribute_34 Out AttributeMathsOp_33 InB)
  (connect Projection_26 Out AttributeMathsOp_35 InA)
  (connect Projection_26 Out AttributeMathsOp_35 InB)
  (connect Gather_10 Out AttributeMathsOp_44 InA)
  (connect Gather_10 Out AttributeMathsOp_44 InB)
  (connect MergePoints_38 Out AttributeMathsOp_48 InA)
  (connect CreateAttribute_49 Out AttributeMathsOp_48 InB)
  (connect AttributeFilter_12 OutsideFilter AttributeMathsOp_51 InA)
  (connect CreateAttribute_53 Out AttributeMathsOp_51 InB)
  (connect AttributeMathsOp_62 Out AttributeMathsOp_57 InA)
  (connect AttributeMathsOp_62 Out AttributeMathsOp_57 InB)
  (connect AttributeMathsOp_57 Out AttributeMathsOp_59 InA)
  (connect AttributeMathsOp_57 Out AttributeMathsOp_59 InB)
  (connect TransformPoints_63 Out AttributeMathsOp_60 InA)
  (connect MergeAttributes_64 Out AttributeMathsOp_60 InB)
  (connect AttributeMathsOp_60 Out AttributeMathsOp_62 InA)
  (connect MergeAttributes_64 Out AttributeMathsOp_62 InB)
  (connect AttributeMathsOp_59 Out AttributeMathsOp_65 InA)
  (connect CreateAttribute_67 Out AttributeMathsOp_65 InB)
  (connect MergePoints_0 Out AttributeMathsOp_71 InA)
  (connect CreateAttribute_72 Out AttributeMathsOp_71 InB)
  (connect CreatePoints_80 Out AttributeMathsOp_76 InA)
  (connect AttributeMathsOp_139 Out AttributeMathsOp_76 InB)
  (connect DeleteAttributes_89 Out AttributeMathsOp_86 InA)
  (connect CreateAttribute_88 Out AttributeMathsOp_86 InB)
  (connect MatchAndSetAttributes_84 Out AttributeMathsOp_89 InA)
  (connect MatchAndSetAttributes_84 Out AttributeMathsOp_89 InB)
  (connect AttributeMathsOp_86 Out AttributeMathsOp_95 InA)
  (connect CreateAttribute_97 Out AttributeMathsOp_95 InB)
  (connect AttributeReduce_14 Out AttributeMathsOp_138 InA)
  (connect AttributeReduce_137 Out AttributeMathsOp_138 InB)
  (connect AttributeMathsOp_95 Out AttributeMathsOp_139 InA)
  (connect ResampleRadius Out AttributeMathsOp_139 InB)
  (connect AttributeMathsOp_17 Out AttributeNoise_12 In)
  (connect GetAttributeFromPointIndex_121 Point AttributeNoise_120 In)
  (connect AttributeFilter_12 InsideFilter AttributePartition_3 In)
  (connect MergePoints_5 Out AttributePartition_4 In)
  (connect CliffSamplesRaw Out AttributePartition_8 In)
  (connect BucketAttributes Out AttributePartition_18 In)
  (connect AttributeNoise_12 Out AttributePartition_19 In)
  (connect MergePoints_11 Out AttributePartition_21 In)
  (connect MergePoints_36 Out AttributePartition_38 In)
  (connect SortAttributes_1 Out AttributePartition_39 In)
  (connect Subgraph_0 ValidEntires AttributePartition_43 In)
  (connect Cliffs Out AttributePartition_56 In)
  (connect MergePoints_57 Out AttributePartition_59 In)
  (connect AttributeFilter_135 InsideFilter AttributePartition_72 In)
  (connect MergePoints_74 Out AttributePartition_76 In)
  (connect AttributeFilter_135 OutsideFilter AttributePartition_145 In)
  (connect MergePoints_158 Out AttributePartition_147 In)
  (connect AttributeFilter_149 InsideFilter AttributePartition_167 In)
  (connect AttributePartition_72 Out AttributeReduce_14 In)
  (connect AttributePartition_21 Out AttributeReduce_22 In)
  (connect AttributeCompareOp_61 Out AttributeReduce_64 In)
  (connect AttributePartition_72 Out AttributeReduce_137 In)
  (connect AttributePartition_167 Out AttributeReduce_151 In)
  (connect AttributePartition_147 Out AttributeReduce_156 In)
  (connect AttributeCast_1 Out AttributeStringOp_2 InA)
  (connect CreateAttribute_3 Out AttributeStringOp_2 InB)
  (connect AttributeStringOp_2 Out AttributeStringOp_5 InA)
  (connect AttributeCast_7 Out AttributeStringOp_5 InB)
  (connect ReprojectMissedTraces Out Branch_8 bOutputToB)
  (connect AttributeMathsOp_0 Out Branch_8 In)
  (connect Debug Out Branch_14 bOutputToB)
  (connect CreatePointsGrid_3 Out Branch_14 In)
  (connect UseDebugRange Out Branch_22 bOutputToB)
  (connect Branch_14 Output%20B Branch_22 In)
  (connect UseSelectiveBuckets Out Branch_27 bOutputToB)
  (connect Branch_22 Output%20A Branch_27 In)
  (connect ResampleCliffs Out Branch_98 bOutputToB)
  (connect Bucket%20Elements Out Branch_98 In)
  (connect ResampleCliffs Out Branch_103 bOutputToB)
  (connect Bucket%20Elements Out Branch_103 In)
  (connect AttributeMathsOp_71 Out CopyAttributes_2 Source)
  (connect AttributeMathsOp_71 Out CopyAttributes_2 Target)
  (connect Buckets Out CopyAttributes_6 Source)
  (connect Buckets Out CopyAttributes_6 Target)
  (connect ExecuteBlueprint_6 Out CopyAttributes_10 Source)
  (connect AttributePartition_8 Out CopyAttributes_10 Target)
  (connect AttributeMathsOp_138 Out CopyAttributes_15 Source)
  (connect GetAttributeFromPointIndex_16 Point CopyAttributes_15 Target)
  (connect AttributeReduce_22 Out CopyAttributes_24 Source)
  (connect AttributePartition_21 Out CopyAttributes_24 Target)
  (connect MergePoints_29 Out CopyAttributes_30 Source)
  (connect MergePoints_29 Out CopyAttributes_30 Target)
  (connect AttributePartition_43 Out CopyAttributes_42 Source)
  (connect AttributePartition_19 Out CopyAttributes_42 Target)
  (connect AttributeReduce_64 Out CopyAttributes_66 Source)
  (connect AttributeCompareOp_61 Out CopyAttributes_66 Target)
  (connect MergePoints_88 Out CopyAttributes_85 Source)
  (connect MergePoints_88 Out CopyAttributes_85 Target)
  (connect AttributeNoise_120 Out CopyAttributes_122 Source)
  (connect AttributeMathsOp_44 Out CopyAttributes_122 Target)
  (connect AttributePartition_167 Out CopyAttributes_163 Source)
  (connect AttributeReduce_151 Out CopyAttributes_163 Target)
  (connect AttributePartition_147 Out CopyAttributes_165 Source)
  (connect AttributeReduce_156 Out CopyAttributes_165 Target)
  (connect MatchAndSetAttributes_173 Out CopyAttributes_175 Source)
  (connect MatchAndSetAttributes_173 Out CopyAttributes_175 Target)
  (connect CopyAttributes_175 Out CopyAttributes_177 Source)
  (connect CopyAttributes_175 Out CopyAttributes_177 Target)
  (connect ExtentsModifier_77 Out CopyPoints_78 Source)
  (connect CopyAttributes_2 Out CopyPoints_78 Target)
  (connect MakeVectorAttribute_10 Out CreatePointsGrid_3 CellSize)
  (connect AttributeMathsOp_13 Out CreatePointsGrid_3 GridExtents)
  (connect Bounds Out DeleteAttributes_7 In)
  (connect AttributeStringOp_5 Out DeleteAttributes_8 In)
  (connect PointToAttributeSet_86 Out DeleteAttributes_87 In)
  (connect FilterElementsByIndex_93 In%20Filter DeleteAttributes_89 In)
  (connect ExtentsModifier_16 Out DensityFilter_24 In)
  (connect AttributeMathsOp_65 Out Difference_68 Differences)
  (connect Branch_103 Output%20B Difference_68 Source)
  (connect BucketAttributes Out ExecuteBlueprint_4 BucketAttributes)
  (connect AttributeMathsOp_19 Out ExecuteBlueprint_4 InPoints)
  (connect GetAttributeFromPointIndex_9 Attribute ExecuteBlueprint_6 InString)
  (connect AttributeCast_31 Out ExecuteBlueprint_30 In)
  (connect MakeVectorAttribute_20 Out ExtentsModifier_16 Extents)
  (connect TransformPoints_25 Out ExtentsModifier_16 In)
  (connect CopyAttributes_177 Out ExtentsModifier_17 In)
  (connect GetAttributeFromPointIndex_43 Point ExtentsModifier_35 In)
  (connect AttributeMathsOp_48 Out ExtentsModifier_45 In)
  (connect CreateAttribute_84 Out ExtentsModifier_77 Extents)
  (connect AttributeMathsOp_76 Out ExtentsModifier_77 In)
  (connect Branch_22 Output%20B FilterElementsByIndex_9 In)
  (connect DeleteAttributes_8 Out FilterElementsByIndex_9 SelectedIndices)
  (connect Branch_27 Output%20B FilterElementsByIndex_25 In)
  (connect ExecuteBlueprint_30 Out FilterElementsByIndex_25 SelectedIndices)
  (connect BucketAttributes Out FilterElementsByIndex_93 In)
  (connect AttributeFilter_1 InsideFilter Gather_4 In)
  (connect AttributePartition_39 Out Gather_10 In)
  (connect Branch_8 Output%20A Gather_10 In)
  (connect Buckets Out GetAttributeFromPointIndex_6 In)
  (connect AttributePartition_4 Out GetAttributeFromPointIndex_8 In)
  (connect AttributePartition_8 Out GetAttributeFromPointIndex_9 In)
  (connect Buckets Out GetAttributeFromPointIndex_11 In)
  (connect AttributePartition_72 Out GetAttributeFromPointIndex_16 In)
  (connect AttributePartition_38 Out GetAttributeFromPointIndex_43 In)
  (connect AttributePartition_59 Out GetAttributeFromPointIndex_79 In)
  (connect AttributePartition_76 Out GetAttributeFromPointIndex_81 In)
  (connect AttributeMathsOp_44 Out GetAttributeFromPointIndex_121 In)
  (connect GetActorData_39 Out GetBounds_41 In)
  (connect AttributePartition_56 Out GetBounds_54 In)
  (connect GetActorData_9 Out GetPropertyFromObjectPath_3 In)
  (connect GetActorData_9 Out GetPropertyFromObjectPath_5 In)
  (connect GetActorData_9 Out GetPropertyFromObjectPath_8 In)
  (connect GetActorData_9 Out GetPropertyFromObjectPath_10 In)
  (connect GetActorData_9 Out GetPropertyFromObjectPath_12 In)
  (connect GetActorData_9 Out GetPropertyFromObjectPath_14 In)
  (connect GetActorData_9 Out GetPropertyFromObjectPath_15 In)
  (connect GetActorData_9 Out GetPropertyFromObjectPath_18 In)
  (connect GetActorData_9 Out GetPropertyFromObjectPath_22 In)
  (connect GetActorData_9 Out GetPropertyFromObjectPath_37 In)
  (connect GetActorData_9 Out GetPropertyFromObjectPath_125 In)
  (connect GetActorData_9 Out GetPropertyFromObjectPath_128 In)
  (connect GetActorData_9 Out GetPropertyFromObjectPath_141 In)
  (connect CopyAttributes_42 Out Loop_24 BoundsPoint)
  (connect MeshDilation Out Loop_24 Dilation)
  (connect MeshResolution Out Loop_24 Resolution)
  (connect CellSize Out MakeVectorAttribute_10 X)
  (connect CellSize Out MakeVectorAttribute_10 Y)
  (connect CreateAttribute_11 Out MakeVectorAttribute_10 Z)
  (connect CreateAttribute_21 Out MakeVectorAttribute_20 X)
  (connect CreateAttribute_21 Out MakeVectorAttribute_20 Y)
  (connect AttributeMathsOp_18 Out MakeVectorAttribute_20 Z)
  (connect MergePoints_82 Out MatchAndSetAttributes_84 In)
  (connect DeleteAttributes_87 Out MatchAndSetAttributes_84 Match%20Data)
  (connect GetAttributeFromPointIndex_16 Point MatchAndSetAttributes_171 In)
  (connect MergeAttributes_178 Out MatchAndSetAttributes_171 Match%20Data)
  (connect MatchAndSetAttributes_171 Out MatchAndSetAttributes_173 In)
  (connect MergeAttributes_180 Out MatchAndSetAttributes_173 Match%20Data)
  (connect GetBounds_54 Out MergeAttributes_64 In)
  (connect CopyAttributes_163 Out MergeAttributes_178 In)
  (connect CopyAttributes_165 Out MergeAttributes_180 In)
  (connect CliffProcessingCandidates Out MergePoints_0 In)
  (connect CopyAttributes_177 Out MergePoints_5 In)
  (connect CopyAttributes_10 Out MergePoints_11 In)
  (connect DensityFilter_24 Out MergePoints_29 In)
  (connect AttributeFilter_26 InsideFilter MergePoints_36 In)
  (connect AttributeFilter_27 OutsideFilter MergePoints_38 In)
  (connect AttributeMathsOp_35 Out MergePoints_38 In)
  (connect AttributeMathsOp_89 Out MergePoints_39 In)
  (connect CopyAttributes_177 Out MergePoints_39 In)
  (connect GetAttributeFromPointIndex_43 Point MergePoints_39 In)
  (connect AttributeFilter_27 OutsideFilter MergePoints_47 In)
  (connect AttributeMathsOp_35 Out MergePoints_47 In)
  (connect Cliff%20Processed Out MergePoints_47 In)
  (connect AttributeFilter_26 OutsideFilter MergePoints_57 In)
  (connect AttributeFilter_68 OutsideFilter MergePoints_70 In)
  (connect AttributeFilter_68 InsideFilter MergePoints_74 In)
  (connect GetAttributeFromPointIndex_81 Point MergePoints_82 In)
  (connect GetAttributeFromPointIndex_79 Point MergePoints_88 In)
  (connect AttributeFilter_149 OutsideFilter MergePoints_158 In)
  (connect AttributePartition_3 Out NormalToDensity_4 In)
  (connect CopyAttributes_6 Out PointToAttributeSet_15 In)
  (connect CopyAttributes_85 Out PointToAttributeSet_86 In)
  (connect AttributeMathsOp_33 Out Projection_26 In)
  (connect CopyAttributes_30 Out Projection_26 Projection%20Target)
  (connect MergePoints_47 Out SortAttributes_1 In)
  (connect GeneratedTextures Out Subgraph_0 In)
  (connect Difference_68 Out ToPoint_70 In)
  (connect Branch_8 Output%20B TransformPoints_25 In)
  (connect Cliffs Out TransformPoints_63 In)
  (connect ResampleCliffs Out WorldRaycast_26 ActorFilterFromInput)
  (connect Cliffs Out WorldRaycast_26 FilterActors)
  (connect Branch_98 Output%20A WorldRaycast_26 Origins)
  (connect CliffProcessingSkip Out WorldRaycast_26 Origins)
  (connect Cliffs Out WorldRaycast_99 FilterActors)
  (connect CopyPoints_78 Out WorldRaycast_99 Origins))
