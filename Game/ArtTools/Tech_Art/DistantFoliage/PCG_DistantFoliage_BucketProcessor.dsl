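; PCG loop subgraph for distant foliage that turns one bounds point and its texture into a grid of bucket points with coverage values.
; It builds a Resolution-based grid over the bounds, derives UVs from each point's offset and samples the texture.
; Clamped alpha becomes Coverage, which is inverted and blurred over a distance scaled inversely with Resolution and offset by Dilation.
; Outputs are points tagged with the loop's BucketIndex, plus the computed CellSize; noise, max and density-copy branches feed no output.
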(graph PCG_DistantFoliage_BucketProcessor
  (params (Resolution :type Int :default 16) (Dilation :type Double))
  (inputs (BoundsPoint :type Point))
  (outputs (Buckets :type Point) (BucketAttributes :type Param))
  (bind CreateAttribute_16 (Create%20Constant :AttributeTypes (struct :DoubleValue 2.0)))
  (bind GetAttributeFromPointIndex_2 (Get%20Attribute%20From%20Point%20Index :bForceOutputAttributeToBeInElementsDomain true :InputSource $BoundsMax :OutputAttributeName @SourceName))
  (bind AttributeMathsOp_14 (Attribute%20Maths%20Op :Operation Multiply))
  (bind CreateAttribute_19 (Create%20Constant :AttributeTypes (struct :DoubleValue 1.0)))
  (bind AttributeMathsOp_17 (Attribute%20Maths%20Op :Operation Subtract :InA Resolution))
  (bind AttributeMathsOp_11 (Attribute%20Maths%20Op :Operation Divide))
  (bind CreateAttribute_8 (Create%20Constant :AttributeTypes (struct :DoubleValue 2.0)))
  (bind AttributeMathsOp_7 (Attribute%20Maths%20Op :InputSource1 BoundsMax.Z :Operation Set))
  (bind CreatePointsGrid_0 (Create%20Points%20Grid :CellSize (1e+02 1e+02 1e+06) :CoordinateSpace OriginalComponent :GridExtents (5e+02 5e+02 1.0) :PointPosition CellCorners :PointSteepness 0.5))
  (bind CreateAttribute_1 (Create%20Constant :AttributeTypes (struct :Type Vector :VectorValue (0.0 0.0 1e+03))))
  (bind AttributeMathsOp_3 (Attribute%20Maths%20Op :InputSource1 $Position))
  (bind AttributeMathsOp_5 (Attribute%20Maths%20Op :InputSource1 $Position :InputSource2 $Position))
  (bind AttributeNoise_6 (Attribute%20Noise))
  (bind CreateAttribute_10 (Create%20Constant :AttributeTypes (struct :DoubleValue 8e+03)))
  (bind AttributeMathsOp_21 (Attribute%20Maths%20Op :Operation Divide :InB Resolution))
  (bind ExtentsModifier_9 (Extents%20Modifier))
  (bind AttributeMathsOp_40 (Attribute%20Maths%20Op :InputSource1 $Position.XY :InputSource2 $Position.XY :Operation Subtract :OutputTarget UVs))
  (bind AttributeMathsOp_41 (Attribute%20Maths%20Op :InputSource1 UVs :InputSource2 $BoundsMax.X :Operation Divide))
  (bind CreateAttribute_44 (Create%20Constant :AttributeTypes (struct :DoubleValue 1.0)))
  (bind AttributeMathsOp_42 (Attribute%20Maths%20Op))
  (bind CreateAttribute_47 (Create%20Constant :AttributeTypes (struct :DoubleValue 2.0)))
  (bind AttributeMathsOp_45 (Attribute%20Maths%20Op :Operation Divide))
  (bind GetAttributeFromPointIndex_28 (Get%20Attribute%20From%20Point%20Index :bForceOutputAttributeToBeInElementsDomain true :OutputAttributeName @SourceName))
  (bind GetTextureData_26 (Get%20Texture%20Data :bForceEditorOnlyCPUSampling true :bUseDensitySourceChannel false))
  (bind SampleTexture_24 (Sample%20Texture :DensityMergeFunction Ignore :TextureMappingMethod UVCoordinates :TilingMode Clamp :UVCoordinatesAttribute UVs))
  (bind CreateAttribute_2 (Create%20Constant))
  (bind CreateAttribute_5 (Create%20Constant :AttributeTypes (struct :DoubleValue 1.0)))
  (bind AttributeMathsOp_0 (Attribute%20Maths%20Op :InputSource1 $Color.A :Operation Clamp :OutputTarget Coverage))
  (bind AttributeMathsOp_8 (Attribute%20Maths%20Op :Operation OneMinus))
  (bind AttributeCast_11 (Attribute%20Cast :In Resolution))
  (bind CreateAttribute_4 (Create%20Constant :AttributeTypes (struct :Type Float :FloatValue 32.0 :Int32Value 32)))
  (bind AttributeMathsOp_1 (Attribute%20Maths%20Op :Operation Divide))
  (bind CreateAttribute_9 (Create%20Constant :AttributeTypes (struct :Type Float :FloatValue 3.2e+03 :Int32Value 3200)))
  (bind AttributeMathsOp_6 (Attribute%20Maths%20Op :Operation Multiply))
  (bind AttributeMathsOp_13 (Attribute%20Maths%20Op :InB Dilation))
  (bind BlurElement_0 (Blur :InputSource Coverage :SearchDistance 3.2e+03))
  (bind AttributeMathsOp_54 (Attribute%20Maths%20Op :InputSource2 $Color.A :Operation Max))
  (bind GetLoopIndex_1 (Get%20Loop%20Index))
  (bind AddAttribute_0 (Add%20Attribute :AttributeTypes (struct :Type Integer32) :OutputTarget BucketIndex))
  (bind CreateAttribute_3 (Create%20Constant :OutputTarget CellSize))
  (bind CopyAttributes_2 (Copy%20Attributes :InputSource BoundsMax.X :OutputTarget CellSize))
  (bind CopyAttributes_9 (Copy%20Attributes :InputSource $Color.A :OutputTarget $Density))
  (reroute CellSize CopyAttributes_2)
  (connect GetLoopIndex_1 Out AddAttribute_0 Attributes)
  (connect BlurElement_0 Out AddAttribute_0 In)
  (connect SampleTexture_24 Out AttributeMathsOp_0 In)
  (connect CreateAttribute_5 Out AttributeMathsOp_0 Max)
  (connect CreateAttribute_2 Out AttributeMathsOp_0 Min)
  (connect CreateAttribute_4 Out AttributeMathsOp_1 InA)
  (connect AttributeCast_11 Out AttributeMathsOp_1 InB)
  (connect CreatePointsGrid_0 Out AttributeMathsOp_3 InA)
  (connect CreateAttribute_1 Out AttributeMathsOp_3 InB)
  (connect AttributeMathsOp_3 Out AttributeMathsOp_5 InA)
  (connect DefaultInputNode BoundsPoint AttributeMathsOp_5 InB)
  (connect AttributeMathsOp_1 Out AttributeMathsOp_6 InA)
  (connect CreateAttribute_9 Out AttributeMathsOp_6 InB)
  (connect AttributeMathsOp_11 Out AttributeMathsOp_7 InA)
  (connect CreateAttribute_8 Out AttributeMathsOp_7 InB)
  (connect AttributeMathsOp_0 Out AttributeMathsOp_8 In)
  (connect AttributeMathsOp_14 Out AttributeMathsOp_11 InA)
  (connect AttributeMathsOp_17 Out AttributeMathsOp_11 InB)
  (connect AttributeMathsOp_6 Out AttributeMathsOp_13 InA)
  (connect GetAttributeFromPointIndex_2 Attribute AttributeMathsOp_14 InA)
  (connect CreateAttribute_16 Out AttributeMathsOp_14 InB)
  (connect CreateAttribute_19 Out AttributeMathsOp_17 InB)
  (connect CreateAttribute_10 Out AttributeMathsOp_21 InA)
  (connect ExtentsModifier_9 Out AttributeMathsOp_40 InA)
  (connect DefaultInputNode BoundsPoint AttributeMathsOp_40 InB)
  (connect AttributeMathsOp_40 Out AttributeMathsOp_41 InA)
  (connect DefaultInputNode BoundsPoint AttributeMathsOp_41 InB)
  (connect AttributeMathsOp_41 Out AttributeMathsOp_42 InA)
  (connect CreateAttribute_44 Out AttributeMathsOp_42 InB)
  (connect AttributeMathsOp_42 Out AttributeMathsOp_45 InA)
  (connect CreateAttribute_47 Out AttributeMathsOp_45 InB)
  (connect BlurElement_0 Out AttributeMathsOp_54 InA)
  (connect BlurElement_0 Out AttributeMathsOp_54 InB)
  (connect AttributeMathsOp_5 Out AttributeNoise_6 In)
  (connect AttributeMathsOp_8 Out BlurElement_0 In)
  (connect AttributeMathsOp_13 Out BlurElement_0 SearchDistance)
  (connect AttributeMathsOp_7 Out CopyAttributes_2 Source)
  (connect CreateAttribute_3 Out CopyAttributes_2 Target)
  (connect SampleTexture_24 Out CopyAttributes_9 Source)
  (connect SampleTexture_24 Out CopyAttributes_9 Target)
  (connect AttributeMathsOp_7 Out CreatePointsGrid_0 CellSize)
  (connect GetAttributeFromPointIndex_2 Attribute CreatePointsGrid_0 GridExtents)
  (connect AttributeMathsOp_21 Out ExtentsModifier_9 Extents)
  (connect AttributeMathsOp_5 Out ExtentsModifier_9 In)
  (connect DefaultInputNode BoundsPoint GetAttributeFromPointIndex_2 In)
  (connect DefaultInputNode BoundsPoint GetAttributeFromPointIndex_28 In)
  (connect GetAttributeFromPointIndex_28 Attribute GetTextureData_26 Texture)
  (connect GetTextureData_26 Out SampleTexture_24 BaseTexture)
  (connect AttributeMathsOp_45 Out SampleTexture_24 Point)
  (output BucketAttributes CellSize)
  (output Buckets AddAttribute_0))
