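; PCG graph that triggers distant foliage generation by calling Generate on actors found through Get Actor Data.
; A Gather node outputs the result of the GetDate Blueprint, with the Generate call as its execution dependency.
; A side branch reads the r.nanite console variable, joins it to a label string and prints it at Display verbosity to record the Nanite setting used during generation.
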(graph PCG_DistantFoliage_Instigator
  (bind ExecuteBlueprint_0 (Execute%20Blueprint :BlueprintElementInstance (class:/Game/Athena/Helios/Environments/Landscape/WorldBuildingTools/PCGGrass/PCGBP_GetDate.PCGBP_GetDate_C) :BlueprintElementType /Game/Athena/Helios/Environments/Landscape/WorldBuildingTools/PCGGrass/PCGBP_GetDate.PCGBP_GetDate_C))
  (bind GetActorData_2 (Get%20Actor%20Data :Mode GetActorReference))
  (bind ApplyOnActor_1 (Apply%20On%20Object :PostProcessFunctionNames [Generate!]))
  (bind Gather_3 (Gather))
  (bind CreateAttribute_2 (Create%20Constant :AttributeTypes (struct :Type String :StringValue "Generating DistantFoliage with r.Nanite = ")))
  (bind ExecuteBlueprint_1 (Execute%20Blueprint :BlueprintElementInstance (class:/Game/Athena/Helios/Environments/Landscape/WorldBuildingTools/PCGGrass/PCGBP_GetCVar.PCGBP_GetCVar_C :CVarName "r.nanite" :CVarType NewEnumerator1) :BlueprintElementType /Game/Athena/Helios/Environments/Landscape/WorldBuildingTools/PCGGrass/PCGBP_GetCVar.PCGBP_GetCVar_C))
  (bind AttributeStringOp_3 (Attribute%20String%20Op))
  (bind Print%20String_0 (Print%20String :Verbosity Display))
  (connect GetActorData_2 Out ApplyOnActor_1 In)
  (connect CreateAttribute_2 Out AttributeStringOp_3 InA)
  (connect ExecuteBlueprint_1 Out AttributeStringOp_3 InB)
  (connect ApplyOnActor_1 Out Gather_3 Execution%20Dependency)
  (connect ExecuteBlueprint_0 Out Gather_3 In)
  (connect ExecuteBlueprint_1 Out Print%20String_0 In)
  (connect AttributeStringOp_3 Out Print%20String_0 PrintString)
  (output Out Gather_3))
