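; PCG graph in the distant foliage tech-art tools that, by its name, prepares mesh data for distant foliage.
; It feeds its input into the custom PCGBP_DistantFoliage_PrepareMesh Blueprint element as InPoints, which holds the logic outside this graph.
; The element's OutPoints pass through a Filter Data By Type node, so only point data reaches the graph output.
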(graph PCG_DistantFoliage_PrepareMeshData
  (bind ExecuteBlueprint_0 (Execute%20Blueprint :BlueprintElementInstance (class:/Game/ArtTools/Tech_Art/DistantFoliage/PCGBP_DistantFoliage_PrepareMesh.PCGBP_DistantFoliage_PrepareMesh_C) :BlueprintElementType /Game/ArtTools/Tech_Art/DistantFoliage/PCGBP_DistantFoliage_PrepareMesh.PCGBP_DistantFoliage_PrepareMesh_C))
  (bind FilterDataByType_1 (Filter%20Data%20By%20Type :TargetType (struct :Ids [(struct :Struct /Script/PCG.PCGDataTypeInfoPoint)])))
  (connect DefaultInputNode In ExecuteBlueprint_0 InPoints)
  (connect ExecuteBlueprint_0 OutPoints FilterDataByType_1 In)
  (output Out FilterDataByType_1.InsideFilter))
