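; PCG graph that prepares per-bucket points for spawning distant foliage from arrays stored on an actor.
; It reads OutputTextures, OutputMeshes and BucketCoordinates, then duplicates a seed point once per texture entry.
; Each point gains Texture and Mesh attributes and takes its position from the bucket coordinates.
; Partitioning by $Index feeds points individually into a looped ActorCreation subgraph, with nothing wired to the graph output.
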(graph PCG_DistantFoliage_Spawner
  (outputs (OutPoints :type Point :advanced))
  (bind GetActorData_2 (Get%20Actor%20Data :Mode GetActorReference))
  (bind GetPropertyFromObjectPath_4 (Get%20Property%20From%20Object%20Path :bSanitizeOutputAttributeName false :OutputAttributeName Texture :PropertyName OutputTextures))
  (bind CreatePoints_6 (Create%20Points :PointsToCreate [(struct)]))
  (bind GetEntriesCount_13 (Get%20Element%20Count))
  (bind DuplicatePoint_12 (Duplicate%20Point :bOutputSourcePoint false :Direction (0.0 0.0 0.0)))
  (bind AddAttribute_15 (Add%20Attribute :bCopyAllAttributes true))
  (bind GetPropertyFromObjectPath_28 (Get%20Property%20From%20Object%20Path :bSanitizeOutputAttributeName false :OutputAttributeName Mesh :PropertyName OutputMeshes))
  (bind AddAttribute_32 (Add%20Attribute :bCopyAllAttributes true))
  (bind GetPropertyFromObjectPath_1 (Get%20Property%20From%20Object%20Path :bSanitizeOutputAttributeName false :OutputAttributeName Position :PropertyName BucketCoordinates))
  (bind CopyAttributes_3 (Copy%20Attributes :InputSource @Last :OutputTarget $Position))
  (bind AttributePartition_5 (Attribute%20Partition :PartitionAttributeSelectors [$Index]))
  (bind Loop_4 (Loop :Graph /Game/ArtTools/Tech_Art/DistantFoliage/PCG_DistantFoliage_Spawner_ActorCreation.PCG_DistantFoliage_Spawner_ActorCreation))
  (connect GetPropertyFromObjectPath_4 Out AddAttribute_15 Attributes)
  (connect DuplicatePoint_12 Out AddAttribute_15 In)
  (connect GetPropertyFromObjectPath_28 Out AddAttribute_32 Attributes)
  (connect AddAttribute_15 Out AddAttribute_32 In)
  (connect CopyAttributes_3 Out AttributePartition_5 In)
  (connect GetPropertyFromObjectPath_1 Out CopyAttributes_3 Source)
  (connect AddAttribute_32 Out CopyAttributes_3 Target)
  (connect CreatePoints_6 Out DuplicatePoint_12 In)
  (connect GetEntriesCount_13 Out DuplicatePoint_12 Iterations)
  (connect GetPropertyFromObjectPath_4 Out GetEntriesCount_13 In)
  (connect GetActorData_2 Out GetPropertyFromObjectPath_1 In)
  (connect GetActorData_2 Out GetPropertyFromObjectPath_4 In)
  (connect GetActorData_2 Out GetPropertyFromObjectPath_28 In)
  (connect AttributePartition_5 Out Loop_4 InPoints))
