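; PCG graph that creates a BP_DistantFoliage actor from the input points, placed at a pivot taken from a point's transform.
; The actor label combines "BP_DistantFoliage" with the Texture attribute, trimmed by PCGBP_TruncateString.
; Point attributes override the actor's Texture and Mesh properties, with Init run as a post-process function.
; A Gather node passes the input points through, using the apply step as an execution dependency.
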(graph PCG_DistantFoliage_Spawner_ActorCreation
  (inputs (InPoints :type Point))
  (outputs (OutPoints :type Point))
  (bind CreateAttribute_2 (Create%20Constant :AttributeTypes (struct :Type String :StringValue "BP_DistantFoliage")))
  (bind PointToAttributeSet_0 (Point%20To%20Attribute%20Set))
  (bind DeleteAttributes_6 (Delete%20Attributes :Operation KeepSelectedAttributes :SelectedAttributes "Texture"))
  (bind ExecuteBlueprint_5 (Execute%20Blueprint :BlueprintElementInstance (class:/Game/Athena/Helios/Environments/Landscape/WorldBuildingTools/PCGGrass/PCGBP_TruncateString.PCGBP_TruncateString_C :ReverseIndex true :TruncationIndex 4) :BlueprintElementType /Game/Athena/Helios/Environments/Landscape/WorldBuildingTools/PCGGrass/PCGBP_TruncateString.PCGBP_TruncateString_C))
  (bind AttributeStringOp_4 (Attribute%20String%20Op))
  (bind GetAttributeFromPointIndex_2 (Get%20Attribute%20From%20Point%20Index :bForceOutputAttributeToBeInElementsDomain true :InputSource $Transform :OutputAttributeName ActorPivot))
  (bind CreateTargetActor_0 (Create%20Target%20Actor :AttachOptions InGraphFolder :TemplateActorClass /Game/ArtTools/Tech_Art/DistantFoliage/BP_DistantFoliage.BP_DistantFoliage_C))
  (bind ApplyOnActor_0 (Apply%20On%20Object :PostProcessFunctionNames [Init] :PropertyOverrideDescriptions [(struct :InputSource Texture :PropertyTarget "Texture") (struct :InputSource Mesh :PropertyTarget "Mesh")]))
  (bind Gather_1 (Gather))
  (connect CreateTargetActor_0 Out ApplyOnActor_0 In)
  (connect PointToAttributeSet_0 Out ApplyOnActor_0 Property%20Overrides)
  (connect CreateAttribute_2 Out AttributeStringOp_4 InA)
  (connect ExecuteBlueprint_5 Out AttributeStringOp_4 InB)
  (connect AttributeStringOp_4 Out CreateTargetActor_0 ActorLabel)
  (connect GetAttributeFromPointIndex_2 Attribute CreateTargetActor_0 ActorPivot)
  (connect PointToAttributeSet_0 Out DeleteAttributes_6 In)
  (connect DeleteAttributes_6 Out ExecuteBlueprint_5 In)
  (connect ApplyOnActor_0 Out Gather_1 Execution%20Dependency)
  (connect DefaultInputNode InPoints Gather_1 In)
  (connect DefaultInputNode InPoints GetAttributeFromPointIndex_2 In)
  (connect DefaultInputNode InPoints PointToAttributeSet_0 In)
  (output OutPoints Gather_1))
